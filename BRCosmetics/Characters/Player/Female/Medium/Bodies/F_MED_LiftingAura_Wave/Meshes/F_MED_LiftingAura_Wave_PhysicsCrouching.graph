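# Crouching-variant physics asset for the LiftingAura Wave medium female body, covering skirt and strap secondary motion.
# Six skirt chains and two strap chains of capsules run from tiny master bodies constrained to the pelvis, with limited swing and twist down each chain.
# Free-rotation cross constraints tie lower segments of skirt columns 01 and 02, plus the left and right 01 tips.
# Pelvis and thigh capsules, including a left-only thigh twist, represent the body.
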PhysicsAsset /BRCosmetics/Characters/Player/Female/Medium/Bodies/F_MED_LiftingAura_Wave/Meshes/F_MED_LiftingAura_Wave_PhysicsCrouching
bodies {
 dyn_skirt_01_01_L [{shapeType=Capsule, length=10, radius=2.5, shapeName=dyn_skirt_front_l_01_capsule, center=<3.5 0 2>, rotation=<rot 0 90 90>}]
 dyn_skirt_01_01_R [{shapeType=Capsule, length=10, radius=2.5, shapeName=dyn_skirt_front_l_01_capsule, center=<3.5 0 2.5>, rotation=<rot 0 0 90>}]
 dyn_skirt_01_02_L [{shapeType=Capsule, length=10, radius=8, shapeName=dyn_skirt_front_l_01_capsule, center=<3.5 0 2>, rotation=<rot 0 90 90>}]
 dyn_skirt_01_02_R [{shapeType=Capsule, length=15, radius=7, shapeName=dyn_skirt_front_l_01_capsule, center=<3.5 0 4>, rotation=<rot 0 0 90>}]
 dyn_skirt_01_03_L [{shapeType=Capsule, length=10, radius=8, shapeName=dyn_skirt_front_l_01_capsule, center=<3.5 0 2>, rotation=<rot 0 0 90>}]
 dyn_skirt_01_03_R [{shapeType=Capsule, length=15.5, radius=7.3, shapeName=dyn_skirt_front_l_01_capsule, center=<3.5 0 3>, rotation=<rot 0 0 90>}]
 dyn_skirt_01_mstr_L [{shapeType=Capsule, length=0.505, radius=0.505, shapeName=dyn_skirt_01_mstr_L_capsule, rotation=<rot 0 -90 0>}]
 dyn_skirt_01_mstr_R [{shapeType=Capsule, length=0.505, radius=0.505, shapeName=dyn_skirt_01_mstr_R_capsule, rotation=<rot 0 -90 0>}]
 dyn_skirt_02_01_L [{shapeType=Capsule, length=14.302, radius=3, shapeName=dyn_skirt_front_l_01_capsule, center=<3.5 0 4>, rotation=<rot 0 0 90>}]
 dyn_skirt_02_01_R [{shapeType=Capsule, length=14.302, radius=3, shapeName=dyn_skirt_front_l_01_capsule, center=<3.5 0 4>, rotation=<rot 0 0 90>}]
 dyn_skirt_02_02_L [{shapeType=Capsule, length=14.302, radius=3.5, shapeName=dyn_skirt_front_l_01_capsule, center=<3.5 0 3>, rotation=<rot 0 0 90>}]
 dyn_skirt_02_02_R [{shapeType=Capsule, length=14.302, radius=4.5, shapeName=dyn_skirt_front_l_01_capsule, center=<3.5 0 3>, rotation=<rot 0 0 90>}]
 dyn_skirt_02_03_L [{shapeType=Capsule, length=10, radius=5, shapeName=dyn_skirt_front_l_01_capsule, center=<3.5 0 3>, rotation=<rot 0 0 90>}]
 dyn_skirt_02_03_R [{shapeType=Capsule, length=20, radius=6, shapeName=dyn_skirt_front_l_01_capsule, center=<3.5 0 3>, rotation=<rot 0 0 90>}]
 dyn_skirt_02_mstr_L [{shapeType=Capsule, length=0.505, radius=0.505, shapeName=dyn_skirt_02_mstr_L_capsule, rotation=<rot 0 -90 0>}]
 dyn_skirt_02_mstr_R [{shapeType=Capsule, length=0.505, radius=0.505, shapeName=dyn_skirt_02_mstr_R_capsule, rotation=<rot 0 -90 0>}]
 dyn_skirt_03_01_L [{shapeType=Capsule, length=10, radius=3.5, shapeName=dyn_skirt_front_l_01_capsule, center=<3.5 0 1.5>, rotation=<rot 0 0 90>}]
 dyn_skirt_03_01_R [{shapeType=Capsule, length=10, radius=3.5, shapeName=dyn_skirt_front_l_01_capsule, center=<3.5 0 1.5>, rotation=<rot 0 0 90>}]
 dyn_skirt_03_02_L [{shapeType=Capsule, length=15, radius=4.5, shapeName=dyn_skirt_front_l_01_capsule, center=<3.5 0 1>, rotation=<rot 0 0 90>}]
 dyn_skirt_03_02_R [{shapeType=Capsule, length=15, radius=4.5, shapeName=dyn_skirt_front_l_01_capsule, center=<3.5 0 1>, rotation=<rot 0 0 90>}]
 dyn_skirt_03_03_L [{shapeType=Capsule, length=15, radius=5, shapeName=dyn_skirt_front_l_01_capsule, center=<3.5 0 0>, rotation=<rot 0 0 90>}]
 dyn_skirt_03_03_R [{shapeType=Capsule, length=15, radius=5.3, shapeName=dyn_skirt_front_l_01_capsule, center=<3.5 0 0>, rotation=<rot 0 0 90>}]
 dyn_skirt_03_mstr_L [{shapeType=Capsule, length=0.505, radius=0.505, shapeName=dyn_skirt_03_mstr_L_capsule, rotation=<rot 0 -90 0>}]
 dyn_skirt_03_mstr_R [{shapeType=Capsule, length=0.505, radius=0.505, shapeName=dyn_skirt_03_mstr_R_capsule, rotation=<rot 0 -90 0>}]
 dyn_strap_01_01 [{shapeType=Capsule, length=4, radius=3, shapeName=dyn_strap_01_01_capsule, center=<5 0 0>, rotation=<rot 90 0 0>}]
 dyn_strap_01_02 [{shapeType=Capsule, length=4, radius=3, shapeName=dyn_strap_01_02_capsule, center=<5 0 0>, rotation=<rot 90 0 0>}]
 dyn_strap_01_03 [{shapeType=Capsule, length=4, radius=4.5, shapeName=dyn_strap_01_03_capsule, center=<5 0 0>, rotation=<rot 90 0 0>}]
 dyn_strap_01_mstr [{shapeType=Capsule, length=0.505, radius=0.505, shapeName=dyn_strap_01_mstr_capsule, center=<1 0 0>, rotation=<rot 90 0 0>}]
 dyn_strap_02_01 [{shapeType=Capsule, length=4, radius=4, shapeName=dyn_strap_02_01_capsule, center=<5 0 3>, rotation=<rot 0 0 90>}]
 dyn_strap_02_02 [{shapeType=Capsule, length=4, radius=5.5, shapeName=dyn_strap_02_02_capsule, center=<5 0 5.2884>, rotation=<rot 0 0 90>}]
 dyn_strap_02_03 [{shapeType=Capsule, length=4, radius=4.5, shapeName=dyn_strap_02_03_capsule, center=<5 0 4.1641>, rotation=<rot 0 0 90>}]
 dyn_strap_02_mstr [{shapeType=Capsule, length=0.505, radius=0.505, shapeName=dyn_strap_02_mstr_capsule, center=<1 0 0>, rotation=<rot 90 0 0>}]
 pelvis [{shapeType=Capsule, length=20, radius=8.731, shapeName=pelvis_capsule, center=<-9.7115 -3.0082 0>}]
 thigh_l [{shapeType=Capsule, length=38.8144, radius=8, shapeName=thigh_l_capsule, center=<-22.9109 -0.1638 0.4468>, rotation=<rot 90 180 -178.5881>}]
 thigh_r [{shapeType=Capsule, length=38.4278, radius=8, shapeName=thigh_l_capsule, center=<22.9109 0.1638 0>, rotation=<rot 90 179.8508 -178.7373>}]
 thigh_twist_01_l [{shapeType=Capsule, length=10, radius=10.5, shapeName=thigh_twist_01_l_capsule, center=<12.2167 -0.7723 0>, rotation=<rot 0 90 90>}]
}
constraints [
 {
  bone1Name=thigh_l
  bone2Name=pelvis
  swing1LimitDegrees=45
  swing1Motion=Free
  swing2LimitDegrees=45
  swing2Motion=Free
  twistLimitDegrees=45
  twistMotion=Free
 }
 {
  bone1Name=thigh_r
  bone2Name=pelvis
  swing1LimitDegrees=45
  swing1Motion=Free
  swing2LimitDegrees=45
  swing2Motion=Free
  twistLimitDegrees=45
  twistMotion=Free
 }
 {bone1Name=dyn_skirt_03_03_R, bone2Name=dyn_skirt_03_02_R, swing1LimitDegrees=25, swing2LimitDegrees=25, twistLimitDegrees=5}
 {
  bone1Name=dyn_skirt_01_mstr_L
  bone2Name=pelvis
  swing1LimitDegrees=45
  swing1Motion=Free
  swing2LimitDegrees=45
  swing2Motion=Free
  twistLimitDegrees=45
  twistMotion=Free
 }
 {bone1Name=dyn_skirt_01_01_L, bone2Name=dyn_skirt_01_mstr_L, swing1LimitDegrees=10, swing2LimitDegrees=25, twistLimitDegrees=5}
 {bone1Name=dyn_skirt_01_02_L, bone2Name=dyn_skirt_01_01_L, swing1LimitDegrees=15, swing2LimitDegrees=35, twistLimitDegrees=15}
 {bone1Name=dyn_skirt_01_03_L, bone2Name=dyn_skirt_01_02_L, swing1LimitDegrees=25, swing2LimitDegrees=40, twistLimitDegrees=5}
 {
  bone1Name=dyn_skirt_01_mstr_R
  bone2Name=pelvis
  swing1LimitDegrees=45
  swing1Motion=Free
  swing2LimitDegrees=45
  swing2Motion=Free
  twistLimitDegrees=45
  twistMotion=Free
 }
 {bone1Name=dyn_skirt_01_01_R, bone2Name=dyn_skirt_01_mstr_R, swing1LimitDegrees=10, swing2LimitDegrees=35, twistLimitDegrees=5}
 {bone1Name=dyn_skirt_01_02_R, bone2Name=dyn_skirt_01_01_R, swing1LimitDegrees=5, swing2LimitDegrees=35, twistLimitDegrees=15}
 {bone1Name=dyn_skirt_01_03_R, bone2Name=dyn_skirt_01_02_R, swing1LimitDegrees=5, swing2LimitDegrees=35, twistLimitDegrees=15}
 {
  bone1Name=dyn_skirt_02_mstr_L
  bone2Name=pelvis
  swing1LimitDegrees=45
  swing1Motion=Free
  swing2LimitDegrees=45
  swing2Motion=Free
  twistLimitDegrees=45
  twistMotion=Free
 }
 {bone1Name=dyn_skirt_02_01_L, bone2Name=dyn_skirt_02_mstr_L, swing1LimitDegrees=10, swing2LimitDegrees=25, twistLimitDegrees=5}
 {bone1Name=dyn_skirt_02_02_L, bone2Name=dyn_skirt_02_01_L, swing1LimitDegrees=25, swing2LimitDegrees=25, twistLimitDegrees=5}
 {bone1Name=dyn_skirt_02_03_L, bone2Name=dyn_skirt_02_02_L, swing1LimitDegrees=25, swing2LimitDegrees=25, twistLimitDegrees=5}
 {
  bone1Name=dyn_skirt_02_mstr_R
  bone2Name=pelvis
  swing1LimitDegrees=45
  swing1Motion=Free
  swing2LimitDegrees=45
  swing2Motion=Free
  twistLimitDegrees=45
  twistMotion=Free
 }
 {bone1Name=dyn_skirt_02_01_R, bone2Name=dyn_skirt_02_mstr_R, swing1LimitDegrees=45, swing2LimitDegrees=25, twistLimitDegrees=5}
 {bone1Name=dyn_skirt_02_02_R, bone2Name=dyn_skirt_02_01_R, swing1LimitDegrees=35, swing2LimitDegrees=35, twistLimitDegrees=5}
 {bone1Name=dyn_skirt_02_03_R, bone2Name=dyn_skirt_02_02_R, swing1LimitDegrees=35, swing2LimitDegrees=35, twistLimitDegrees=5}
 {
  bone1Name=dyn_skirt_03_mstr_L
  bone2Name=pelvis
  swing1LimitDegrees=45
  swing1Motion=Free
  swing2LimitDegrees=45
  swing2Motion=Free
  twistLimitDegrees=45
  twistMotion=Free
 }
 {bone1Name=dyn_skirt_03_01_L, bone2Name=dyn_skirt_03_mstr_L, swing1LimitDegrees=10, swing2LimitDegrees=20, twistLimitDegrees=5}
 {bone1Name=dyn_skirt_03_02_L, bone2Name=dyn_skirt_03_01_L, swing1LimitDegrees=25, swing2LimitDegrees=25, twistLimitDegrees=5}
 {bone1Name=dyn_skirt_03_03_L, bone2Name=dyn_skirt_03_02_L, swing1LimitDegrees=25, swing2LimitDegrees=25, twistLimitDegrees=5}
 {
  bone1Name=dyn_skirt_03_mstr_R
  bone2Name=pelvis
  swing1LimitDegrees=45
  swing1Motion=Free
  swing2LimitDegrees=45
  swing2Motion=Free
  twistLimitDegrees=45
  twistMotion=Free
 }
 {bone1Name=dyn_skirt_03_01_R, bone2Name=dyn_skirt_03_mstr_R, swing1LimitDegrees=10, swing2LimitDegrees=20, twistLimitDegrees=5}
 {bone1Name=dyn_skirt_03_02_R, bone2Name=dyn_skirt_03_01_R, swing1LimitDegrees=25, swing2LimitDegrees=25, twistLimitDegrees=5}
 {
  bone1Name=dyn_skirt_01_02_L
  bone2Name=dyn_skirt_02_02_L
  swing1LimitDegrees=45
  swing1Motion=Free
  swing2LimitDegrees=45
  swing2Motion=Free
  twistLimitDegrees=45
  twistMotion=Free
 }
 {bone1Name=dyn_strap_02_03, bone2Name=dyn_strap_02_02, swing1LimitDegrees=40, swing2LimitDegrees=40, twistLimitDegrees=20}
 {
  bone1Name=dyn_strap_02_mstr
  bone2Name=pelvis
  swing1LimitDegrees=45
  swing1Motion=Free
  swing2LimitDegrees=45
  swing2Motion=Free
  twistLimitDegrees=45
  twistMotion=Free
 }
 {bone1Name=dyn_strap_02_01, bone2Name=dyn_strap_02_mstr, swing1LimitDegrees=30, swing2LimitDegrees=30, twistLimitDegrees=20}
 {bone1Name=dyn_strap_02_02, bone2Name=dyn_strap_02_01, swing1LimitDegrees=40, swing2LimitDegrees=40, twistLimitDegrees=20}
 {bone1Name=dyn_strap_01_03, bone2Name=dyn_strap_01_02, swing1LimitDegrees=45, swing2LimitDegrees=45, twistLimitDegrees=10}
 {bone1Name=dyn_strap_01_02, bone2Name=dyn_strap_01_01, swing1LimitDegrees=45, swing2LimitDegrees=45, twistLimitDegrees=10}
 {bone1Name=dyn_strap_01_01, bone2Name=dyn_strap_01_mstr, swing1LimitDegrees=20, swing2LimitDegrees=20, twistLimitDegrees=10}
 {
  bone1Name=dyn_strap_01_mstr
  bone2Name=pelvis
  swing1LimitDegrees=45
  swing1Motion=Free
  swing2LimitDegrees=45
  swing2Motion=Free
  twistLimitDegrees=45
  twistMotion=Free
 }
 {
  bone1Name=dyn_skirt_01_02_R
  bone2Name=dyn_skirt_02_02_R
  swing1LimitDegrees=45
  swing1Motion=Free
  swing2LimitDegrees=45
  swing2Motion=Free
  twistLimitDegrees=45
  twistMotion=Free
 }
 {
  bone1Name=dyn_skirt_01_03_L
  bone2Name=dyn_skirt_02_03_L
  swing1LimitDegrees=45
  swing1Motion=Free
  swing2LimitDegrees=45
  swing2Motion=Free
  twistLimitDegrees=45
  twistMotion=Free
 }
 {
  bone1Name=dyn_skirt_01_03_R
  bone2Name=dyn_skirt_01_03_L
  swing1LimitDegrees=45
  swing1Motion=Free
  swing2LimitDegrees=45
  swing2Motion=Free
  twistLimitDegrees=45
  twistMotion=Free
 }
 {
  bone1Name=dyn_skirt_01_03_L
  bone2Name=dyn_skirt_01_03_R
  swing1LimitDegrees=45
  swing1Motion=Free
  swing2LimitDegrees=45
  swing2Motion=Free
  twistLimitDegrees=45
  twistMotion=Free
 }
 {
  bone1Name=dyn_skirt_01_03_R
  bone2Name=dyn_skirt_02_03_R
  swing1LimitDegrees=45
  swing1Motion=Free
  swing2LimitDegrees=45
  swing2Motion=Free
  twistLimitDegrees=45
  twistMotion=Free
 }
 {
  bone1Name=thigh_twist_01_l
  bone2Name=thigh_l
  swing1LimitDegrees=45
  swing1Motion=Free
  swing2LimitDegrees=45
  swing2Motion=Free
  twistLimitDegrees=45
  twistMotion=Free
 }
]
