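# Anim Blueprint for the LiftingAura_Wave female medium body cosmetic, adding secondary motion to its skirt, straps and jacket.
# Branches from cached poses run pelvis-space RigidBody sims with separate physics assets, including skydive, cape, crouching and emote variants, merged through timed bool blends.
# Constraint and additive bone-transform chains pin straps and skirt to pelvis and thighs, correcting individual branches and the blended result.
# Spring-based AnimDynamics with spherical limits drives jacket bones and a strap tip.
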AnimBlueprint /BRCosmetics/Characters/Player/Female/Medium/Bodies/F_MED_LiftingAura_Wave/Meshes/F_MED_LiftingAura_Wave_AnimBP
target_skeleton=&/F_MED_LiftingAura_Wave_Skeleton
graphs [
 AnimGraph {nodes [
   n1 = RigidBody(ComponentPose=Local To Component(LocalPose=Use cached pose 'PreSim'())) {
    bOverrideWorldGravity=true
    overridePhysicsAsset=&/F_MED_LiftingAura_Wave_Physics_SkyDive
    simulationSpace=BaseBoneSpace
    alphaBoolBlend {blendOption=Linear}
    alphaScaleBiasClamp {bInterpResult=true}
    baseBoneRef {boneName=pelvis}
    componentAppliedLinearAccClamp=<1500 1500 1500>
    componentLinearAccScale=<0.5 0.5 0.5>
    componentLinearVelScale=<0.5 0.5 0.5>
    overrideWorldGravity {z=-980}
    simSpaceSettings {maxAngularAcceleration=250, maxLinearAcceleration=250, maxLinearVelocity=500, velocityScaleZ=0, worldAlpha=0.5}
   }
   n2 = Component To Local(ComponentPose=@n1)
   n3 = RigidBody(ComponentPose=Local To Component(LocalPose=Use cached pose 'PreSim'()), Alpha=Get SkyDiveNot() -> (SkyDiveNot:double)) {
    bOverrideWorldGravity=true
    overridePhysicsAsset=&/F_MED_LiftingAura_Wave_Physics
    simulationSpace=BaseBoneSpace
    alphaBoolBlend {blendOption=Linear}
    alphaScaleBiasClamp {bInterpResult=true}
    baseBoneRef {boneName=pelvis}
    componentAppliedLinearAccClamp=<1500 1500 1500>
    componentLinearAccScale=<0.5 0.5 0.5>
    componentLinearVelScale=<0.6 0.6 0.6>
    overrideWorldGravity {z=-980}
    simSpaceSettings {maxAngularAcceleration=250, maxLinearAcceleration=250, maxLinearVelocity=500, worldAlpha=1}
   }
   n4 = Component To Local(ComponentPose=@n3)
   n5 = Get bIsSkyDiving()
   n6 = RigidBody(ComponentPose=Local To Component(LocalPose=Use cached pose 'PreSim'())) {
    bOverrideWorldGravity=true
    overridePhysicsAsset=&/F_MED_LiftingAura_Wave_PhysicsCape2
    simulationSpace=BaseBoneSpace
    alphaBoolBlend {blendOption=Linear}
    alphaScaleBiasClamp {bInterpResult=true}
    baseBoneRef {boneName=pelvis}
    componentAppliedLinearAccClamp=<1500 1500 1500>
    componentLinearAccScale=<0.5 0.5 0.5>
    componentLinearVelScale=<0.6 0.6 0.6>
    overrideWorldGravity {z=-980}
    simSpaceSettings {maxAngularAcceleration=250, maxLinearAcceleration=250, maxLinearVelocity=500, worldAlpha=1}
   }
   n7 = "Transform (Modify) Bone - Bone: dyn_strap_02_mstr"!(ComponentPose=Local To Component(LocalPose=Use cached pose 'Pre-Dynamics'()), Alpha=1, Rotation="0,90,60") -> (Pose) {
    rotationMode=BMM_Additive
    rotationSpace=BCS_BoneSpace
    translationMode=BMM_Additive
    translationSpace=BCS_BoneSpace
    alphaBoolBlend {blendOption=Linear}
    boneToModify {boneName=dyn_strap_02_mstr}
    rotation {roll=60, yaw=90}
   }
   n8 = "Transform (Modify) Bone - Bone: dyn_strap_02_mstr"(ComponentPose=@n7, Alpha=1, Rotation="49.031,-44.212,39.389") {
    rotationMode=BMM_Additive
    rotationSpace=BCS_BoneSpace
    translationMode=BMM_Additive
    translationSpace=BCS_BoneSpace
    alphaBoolBlend {blendOption=Linear}
    boneToModify {boneName=dyn_strap_02_mstr}
    rotation=<rot 49.0311 -44.2121 39.3889>
    translation {x=2}
   }
   n9 = "Constraint - dyn_strap_02_03 (PPPPPPPP)"!(
    ComponentPose=@n8
    Alpha=1
    ConstraintWeights_0=0.0673
    ConstraintWeights_1:float=0.1578
    ConstraintWeights_2:float=0.2811
    ConstraintWeights_3:float=0.177
    ConstraintWeights_4:float=0.0186
    ConstraintWeights_5:float=0.2128
    ConstraintWeights_6:float=0.0727
    ConstraintWeights_7:float=0.0127
   ) -> (Pose) {
    alphaBoolBlend {blendOption=Linear}
    boneToModify {boneName=dyn_strap_02_03}
    constraintSetup [
     {targetBone {boneName=pelvis}}
     {targetBone {boneName=thigh_twist_01_l}}
     {targetBone {boneName=dyn_skirt_02_01_L}}
     {targetBone {boneName=dyn_skirt_02_02_L}}
     {targetBone {boneName=dyn_skirt_02_03_L}}
     {targetBone {boneName=dyn_skirt_03_01_L}}
     {targetBone {boneName=dyn_skirt_03_02_L}}
     {targetBone {boneName=dyn_skirt_03_03_L}}
    ]
    constraintWeights [1,1,1,1,1,1,1,1]
   }
   n10 = "Constraint - dyn_strap_02_02 (PPPP)"!(
    ComponentPose=@n9
    Alpha=1
    ConstraintWeights_0=0.6864
    ConstraintWeights_1:float=0.1553
    ConstraintWeights_2:float=0.0903
    ConstraintWeights_3:float=0.0679
   ) -> (Pose) {
    alphaBoolBlend {blendOption=Linear}
    boneToModify {boneName=dyn_strap_02_02}
    constraintSetup [
     {targetBone {boneName=pelvis}}
     {targetBone {boneName=thigh_twist_01_l}}
     {targetBone {boneName=deform_groin_l}}
     {targetBone {boneName=dyn_skirt_02_01_L}}
    ]
    constraintWeights [1,1,1,1]
   }
   n11 = RigidBody(ComponentPose=@n10) {
    bOverrideWorldGravity=true
    overridePhysicsAsset=&/F_MED_LiftingAura_Wave_PhysicsCrouching
    simulationSpace=BaseBoneSpace
    alphaBoolBlend {blendOption=Linear}
    alphaScaleBiasClamp {bInterpResult=true}
    baseBoneRef {boneName=pelvis}
    componentAppliedLinearAccClamp=<1000 1000 1000>
    componentLinearAccScale=<0.7 0.7 0.7>
    componentLinearVelScale=<0.6 0.6 0.6>
    overrideWorldGravity {z=-980}
    simSpaceSettings {maxAngularAcceleration=250, maxLinearAcceleration=250, maxLinearVelocity=500, velocityScaleZ=0, worldAlpha=1}
   }
   n12 = "Constraint - dyn_strap_01_01 (RR)"!(ComponentPose=@n11, Alpha=1, ConstraintWeights_0=0.5, ConstraintWeights_1:float=0.5) -> (Pose) {
    alphaBoolBlend {blendOption=Linear}
    boneToModify {boneName=dyn_strap_01_01}
    constraintSetup [{transformType=Rotation, targetBone {boneName=dyn_skirt_01_01_L}}, {transformType=Rotation, targetBone {boneName=thigh_l}}]
    constraintWeights [1,1]
   }
   n13 = "Transform (Modify) Bone - Bone: dyn_strap_01_01"(ComponentPose=@n12, Alpha=1, Translation="2,4,0", Rotation="0,-10,60", Scale="1,1,1") {
    rotationMode=BMM_Additive
    rotationSpace=BCS_BoneSpace
    translationMode=BMM_Additive
    translationSpace=BCS_BoneSpace
    alphaBoolBlend {blendOption=Linear}
    boneToModify {boneName=dyn_strap_01_01}
    rotation {roll=60, yaw=-10}
    translation=<2 4>
   }
   n14 = "Constraint - dyn_strap_01_02 (RR)"!(ComponentPose=@n13, Alpha=1, ConstraintWeights_0=0.5, ConstraintWeights_1:float=0.5) -> (Pose) {
    alphaBoolBlend {blendOption=Linear}
    boneToModify {boneName=dyn_strap_01_02}
    constraintSetup [{transformType=Rotation, targetBone {boneName=dyn_skirt_01_01_L}}, {transformType=Rotation, targetBone {boneName=thigh_l}}]
    constraintWeights [1,1]
   }
   n15 = "Transform (Modify) Bone - Bone: dyn_strap_01_02"(ComponentPose=@n14, Alpha=1, Translation="0,0,0", Rotation="0,-10,50", Scale="1,1,1") {
    rotationMode=BMM_Additive
    rotationSpace=BCS_BoneSpace
    translationMode=BMM_Additive
    translationSpace=BCS_BoneSpace
    alphaBoolBlend {blendOption=Linear}
    boneToModify {boneName=dyn_strap_01_02}
    rotation {roll=50, yaw=-10}
   }
   n16 = Anim Dynamics - Bone: dyn_strap_01_03!(ComponentPose=@n15, Alpha=1, GravityOverride:Vector="0,0,0", ExternalForce="0,0,0") -> (Pose) {
    angularDampingOverride=0.9999
    angularSpringConstant=500
    bAngularSpring=true
    bGravityOverrideInSimSpace=true
    bOverrideAngularDamping=true
    bUseGravityOverride=true
    bUsePlanarLimit=false
    simulationSpace=BoneRelative
    alphaBoolBlend {blendOption=Linear}
    boundBone {boneName=dyn_strap_01_03}
    componentAppliedLinearAccClamp=<3200 3200 3200>
    componentLinearAccScale=<1.2 1.2 1.2>
    componentLinearVelScale=<1.2 1.2 1.2>
    physicsBodyDefinitions [{
      collisionType=CustomSphere
      sphereCollisionRadius=1
      boundBone {boneName=dyn_strap_01_03}
      boxExtents=<15 20 20>
      constraintSetup {angularLimitsMax=<3 3 10>, angularLimitsMin=<-3 -3 -10>, angularTarget=<0 0 0>}
      localJointOffset=<3 0 0>
     }]
    planarLimits [{drivingBone {boneName=thigh_twist_01_l}, planeTransform {location=<0 8 0>, rotation=<rot 0 0 94.9999>}}]
    relativeSpaceBone {boneName=root}
    sphericalLimits [
     {limitRadius=9.3, drivingBone {boneName=thigh_l}, sphereLocalOffset=<-6 0 0>}
     {limitRadius=11, drivingBone {boneName=thigh_l}, sphereLocalOffset=<-12 0 0>}
    ]
   }
   n17 = Blend Poses by bool(
    BlendPose_0=Component To Local(ComponentPose=@n16)
    BlendPose_1=Blend Poses by bool(
     BlendPose_0=Component To Local(ComponentPose=@n6)
     BlendPose_1=Blend Poses by bool(BlendPose_0=@n2, BlendPose_1=@n4, BlendTime_0=0.2, BlendTime_1=0.2, bActiveValue=@n5) {blendType=Linear}
     BlendTime_0=0.2
     BlendTime_1=0.2
     bActiveValue=Get CapeOn() -> (CapeOn:bool)
    ) {blendType=Linear}
    BlendTime_0=0.2
    BlendTime_1=0.2
   ) {blendType=Linear}
   n18 = RigidBody(ComponentPose=Local To Component(LocalPose=Use cached pose 'Pre-Dynamics'())) {
    bOverrideWorldGravity=true
    overridePhysicsAsset=&/F_MED_LiftingAura_Wave_Physics_FE
    simulationSpace=BaseBoneSpace
    alphaBoolBlend {blendOption=Linear}
    alphaScaleBiasClamp {bInterpResult=true}
    baseBoneRef {boneName=pelvis}
    componentAppliedLinearAccClamp=<1500 1500 1500>
    componentLinearAccScale=<0.5 0.5 0.5>
    componentLinearVelScale=<0.6 0.6 0.6>
    overrideWorldGravity {z=-980}
    simSpaceSettings {maxAngularAcceleration=250, maxLinearAcceleration=250, maxLinearVelocity=500, worldAlpha=1}
   }
   n19 = "Constraint - dyn_strap_02_01 (PR)"!(ComponentPose=@n18, Alpha=1, ConstraintWeights_0=1, ConstraintWeights_1:float=0) -> (Pose) {
    alphaBoolBlend {blendOption=Linear}
    alphaScaleBiasClamp {bInterpResult=true, interpSpeedDecreasing=1, interpSpeedIncreasing=1}
    boneToModify {boneName=dyn_strap_02_01}
    constraintSetup [{targetBone {boneName=dyn_skirt_01_01_L}}, {transformType=Rotation, targetBone {boneName=thigh_l}}]
    constraintWeights [1,1]
   }
   n20 = "Transform (Modify) Bone - Bone: dyn_strap_02_01"(ComponentPose=@n19, Alpha=1, Translation="2.914,8.042,-0.416", Rotation="9.163,8.352,42.672", Scale="1,1,1") {
    rotationMode=BMM_Additive
    rotationSpace=BCS_BoneSpace
    translationMode=BMM_Additive
    translationSpace=BCS_BoneSpace
    alphaBoolBlend {blendOption=Linear}
    alphaScaleBiasClamp {bInterpResult=true, interpSpeedDecreasing=1, interpSpeedIncreasing=1}
    boneToModify {boneName=dyn_strap_02_01}
    rotation=<rot 9.163 8.352 42.672>
    translation=<2.914 8.042 -0.416>
   }
   n21 = "Constraint - dyn_strap_01_01 (PR)"!(ComponentPose=@n20, Alpha=1, ConstraintWeights_0=1, ConstraintWeights_1:float=0) -> (Pose) {
    alphaBoolBlend {blendOption=Linear}
    alphaScaleBiasClamp {bInterpResult=true, interpSpeedDecreasing=1, interpSpeedIncreasing=1}
    boneToModify {boneName=dyn_strap_01_01}
    constraintSetup [{targetBone {boneName=dyn_skirt_01_01_L}}, {transformType=Rotation, targetBone {boneName=thigh_l}}]
    constraintWeights [1,1]
   }
   n22 = "Transform (Modify) Bone - Bone: dyn_strap_01_01"(ComponentPose=@n21, Alpha=1, Translation="5.335,2.743,3.055", Rotation="1.389,-10.285,36.406", Scale="1,1,1") {
    rotationMode=BMM_Additive
    rotationSpace=BCS_BoneSpace
    translationMode=BMM_Additive
    translationSpace=BCS_BoneSpace
    alphaBoolBlend {blendOption=Linear}
    alphaScaleBiasClamp {bInterpResult=true, interpSpeedDecreasing=1, interpSpeedIncreasing=1}
    boneToModify {boneName=dyn_strap_01_01}
    rotation=<rot 1.389 -10.285 36.406>
    translation=<5.335 2.743 3.055>
   }
   n23 = Component To Local(ComponentPose=@n22)
   n24 = Anim Dynamics - Bone: dyn_jacket_L_01!(
    ComponentPose=Local To Component(LocalPose=Blend Poses by bool(BlendPose_0=@n23, BlendPose_1=@n17, BlendTime_0=0.2, BlendTime_1=0.2) {blendType=Linear})
    Alpha=1
    GravityOverride:Vector="0,0,0"
    ExternalForce="0,0,0"
   ) -> (Pose) {
    angularDampingOverride=0.999
    angularSpringConstant=500
    bAngularSpring=true
    bGravityOverrideInSimSpace=true
    bOverrideAngularDamping=true
    bUseGravityOverride=true
    simulationSpace=BoneRelative
    alphaBoolBlend {blendOption=Linear}
    boundBone {boneName=dyn_jacket_L_01}
    componentAppliedLinearAccClamp=<3200 3200 3200>
    componentLinearAccScale=<1.5 1.5 1.5>
    componentLinearVelScale=<1.5 1.5 1.5>
    physicsBodyDefinitions [{
      collisionType=CustomSphere
      sphereCollisionRadius=1
      boundBone {boneName=dyn_jacket_L_01}
      boxExtents=<15 20 20>
      constraintSetup {angularLimitsMax=<1 1 15>, angularLimitsMin=<-1 -15 -15>, angularTarget=<0 0 0>}
      localJointOffset=<3 0 0>
     }]
    relativeSpaceBone {boneName=root}
    sphericalLimits [
     {limitRadius=10, drivingBone {boneName=spine_03}, sphereLocalOffset=<0 0 -2>}
     {limitRadius=9, drivingBone {boneName=spine_04}, sphereLocalOffset=<0 0 -2>}
    ]
   }
   n25 = Anim Dynamics - Bone: dyn_jacket_LS_01!(ComponentPose=@n24, Alpha=1, GravityOverride:Vector="0,0,0", ExternalForce="0,0,0") -> (Pose) {
    angularDampingOverride=0.999
    angularSpringConstant=500
    bAngularSpring=true
    bGravityOverrideInSimSpace=true
    bOverrideAngularDamping=true
    bUseGravityOverride=true
    simulationSpace=BoneRelative
    alphaBoolBlend {blendOption=Linear}
    boundBone {boneName=dyn_jacket_LS_01}
    componentAppliedLinearAccClamp=<3200 3200 3200>
    componentLinearAccScale=<1.5 1.5 1.5>
    componentLinearVelScale=<1.5 1.5 1.5>
    physicsBodyDefinitions [{
      collisionType=CustomSphere
      sphereCollisionRadius=1
      boundBone {boneName=dyn_jacket_LS_01}
      boxExtents=<15 20 20>
      constraintSetup {angularLimitsMax=<1 1 15>, angularLimitsMin=<-1 -15 -15>, angularTarget=<0 0 0>}
      localJointOffset=<3 0 0>
     }]
    relativeSpaceBone {boneName=root}
    sphericalLimits [
     {limitRadius=10, drivingBone {boneName=spine_03}, sphereLocalOffset=<0 0 -2>}
     {limitRadius=9, drivingBone {boneName=spine_04}, sphereLocalOffset=<0 0 -2>}
    ]
   }
   n26 = Anim Dynamics - Bone: dyn_jacket_B_01!(ComponentPose=@n25, Alpha=1, GravityOverride:Vector="0,0,0", ExternalForce="0,0,0") -> (Pose) {
    angularDampingOverride=0.999
    angularSpringConstant=500
    bAngularSpring=true
    bGravityOverrideInSimSpace=true
    bOverrideAngularDamping=true
    bUseGravityOverride=true
    simulationSpace=BoneRelative
    alphaBoolBlend {blendOption=Linear}
    boundBone {boneName=dyn_jacket_B_01}
    componentAppliedLinearAccClamp=<3200 3200 3200>
    componentLinearAccScale=<1.5 1.5 1.5>
    componentLinearVelScale=<1.5 1.5 1.5>
    physicsBodyDefinitions [{
      collisionType=CustomSphere
      sphereCollisionRadius=1
      boundBone {boneName=dyn_jacket_B_01}
      boxExtents=<15 20 20>
      constraintSetup {angularLimitsMax=<1 1 15>, angularLimitsMin=<-1 -15 -15>, angularTarget=<0 0 0>}
      localJointOffset=<3 0 0>
     }]
    relativeSpaceBone {boneName=root}
    sphericalLimits [
     {limitRadius=10, drivingBone {boneName=spine_03}, sphereLocalOffset=<0 0 -2>}
     {limitRadius=9, drivingBone {boneName=spine_04}, sphereLocalOffset=<0 0 -2>}
    ]
   }
   n27 = Anim Dynamics - Bone: dyn_jacket_R_01!(ComponentPose=@n26, Alpha=1, GravityOverride:Vector="0,0,0", ExternalForce="0,0,0") -> (Pose) {
    angularDampingOverride=0.999
    angularSpringConstant=500
    bAngularSpring=true
    bGravityOverrideInSimSpace=true
    bOverrideAngularDamping=true
    bUseGravityOverride=true
    simulationSpace=BoneRelative
    alphaBoolBlend {blendOption=Linear}
    boundBone {boneName=dyn_jacket_R_01}
    componentAppliedLinearAccClamp=<3200 3200 3200>
    componentLinearAccScale=<1.5 1.5 1.5>
    componentLinearVelScale=<1.5 1.5 1.5>
    physicsBodyDefinitions [{
      collisionType=CustomSphere
      sphereCollisionRadius=1
      boundBone {boneName=dyn_jacket_R_01}
      boxExtents=<15 20 20>
      constraintSetup {angularLimitsMax=<1 1 15>, angularLimitsMin=<-1 -15 -15>, angularTarget=<0 0 0>}
      localJointOffset=<3 0 0>
     }]
    relativeSpaceBone {boneName=root}
    sphericalLimits [
     {limitRadius=10, drivingBone {boneName=spine_03}, sphereLocalOffset=<0 0 -2>}
     {limitRadius=9, drivingBone {boneName=spine_04}, sphereLocalOffset=<0 0 -2>}
    ]
   }
   n28 = Anim Dynamics - Bone: dyn_jacket_RS_01!(ComponentPose=@n27, Alpha=1, GravityOverride:Vector="0,0,0", ExternalForce="0,0,0") -> (Pose) {
    angularDampingOverride=0.999
    angularSpringConstant=500
    bAngularSpring=true
    bGravityOverrideInSimSpace=true
    bOverrideAngularDamping=true
    bUseGravityOverride=true
    simulationSpace=BoneRelative
    alphaBoolBlend {blendOption=Linear}
    boundBone {boneName=dyn_jacket_RS_01}
    componentAppliedLinearAccClamp=<3200 3200 3200>
    componentLinearAccScale=<1.5 1.5 1.5>
    componentLinearVelScale=<1.5 1.5 1.5>
    physicsBodyDefinitions [{
      collisionType=CustomSphere
      sphereCollisionRadius=1
      boundBone {boneName=dyn_jacket_RS_01}
      boxExtents=<15 20 20>
      constraintSetup {angularLimitsMax=<1 1 15>, angularLimitsMin=<-1 -15 -15>, angularTarget=<0 0 0>}
      localJointOffset=<3 0 0>
     }]
    relativeSpaceBone {boneName=root}
    sphericalLimits [
     {limitRadius=10, drivingBone {boneName=spine_03}, sphereLocalOffset=<0 0 -2>}
     {limitRadius=9, drivingBone {boneName=spine_04}, sphereLocalOffset=<0 0 -2>}
    ]
   }
   n29 = "Constraint - dyn_strap_02_01 (P)"(ComponentPose=Local To Component(LocalPose=Use cached pose 'Pre-Dynamics'()), Alpha=1, ConstraintWeights_0=1) {
    alphaBoolBlend {blendOption=Linear}
    boneToModify {boneName=dyn_strap_02_01}
    constraintSetup [{targetBone {boneName=thigh_l}}]
    constraintWeights [1]
   }
   n30 = "Constraint - dyn_strap_02_02 (P)"(ComponentPose=@n29, Alpha=0.5, ConstraintWeights_0=1) {
    alphaBoolBlend {blendOption=Linear}
    boneToModify {boneName=dyn_strap_02_02}
    constraintSetup [{targetBone {boneName=dyn_skirt_02_01_L}}]
    constraintWeights [1]
   }
   n31 = RigidBody(ComponentPose=@n30) {
    bOverrideWorldGravity=true
    overridePhysicsAsset=&/F_MED_LiftingAura_Wave_PhysicsEmote
    simulationSpace=BaseBoneSpace
    alphaBoolBlend {blendOption=Linear}
    alphaScaleBiasClamp {bInterpResult=true}
    baseBoneRef {boneName=pelvis}
    componentAppliedLinearAccClamp=<1500 1500 1500>
    componentLinearAccScale=<0.5 0.5 0.5>
    componentLinearVelScale=<0.6 0.6 0.6>
    overrideWorldGravity {z=-980}
    simSpaceSettings {maxAngularAcceleration=250, maxLinearAcceleration=250, maxLinearVelocity=500, velocityScaleZ=0, worldAlpha=1}
   }
   n32 = "Transform (Modify) Bone - Bone: dyn_skirt_01_01_L"(ComponentPose=@n31, Alpha=1, Translation="0,0,10", Rotation="0,0,0", Scale="1,1,1") {
    rotationMode=BMM_Additive
    translationMode=BMM_Additive
    alphaBoolBlend {blendOption=Linear}
    boneToModify {boneName=dyn_skirt_01_01_L}
    translation {z=10}
   }
   n33 = "Transform (Modify) Bone - Bone: dyn_skirt_01_01_R"(ComponentPose=@n32, Alpha=1, Translation="0,0,10", Rotation="0,0,0", Scale="1,1,1") {
    rotationMode=BMM_Additive
    translationMode=BMM_Additive
    alphaBoolBlend {blendOption=Linear}
    boneToModify {boneName=dyn_skirt_01_01_R}
    translation {z=10}
   }
   n34 = "Transform (Modify) Bone - Bone: dyn_skirt_03_01_L"(ComponentPose=@n33, Alpha=1, Translation="0,0,2", Rotation="0,0,0", Scale="1,1,1") {
    rotationMode=BMM_Additive
    translationMode=BMM_Additive
    alphaBoolBlend {blendOption=Linear}
    boneToModify {boneName=dyn_skirt_03_01_L}
    translation {z=2}
   }
   n35 = "Constraint - dyn_skirt_01_02_L (RR)"!(ComponentPose=@n34, Alpha=1, ConstraintWeights_0=0.8, ConstraintWeights_1:float=0.2) -> (Pose) {
    alphaBoolBlend {blendOption=Linear}
    boneToModify {boneName=dyn_skirt_01_02_L}
    constraintSetup [{transformType=Rotation, targetBone {boneName=thigh_l}}, {transformType=Rotation, targetBone {boneName=pelvis}}]
    constraintWeights [1,1]
   }
   n36 = "Transform (Modify) Bone - Bone: dyn_skirt_01_02_L"(ComponentPose=@n35, Alpha=1, Translation="0,0,0", Rotation="0,0,0", Scale="1,1,1") {rotationMode=BMM_Additive, translationMode=BMM_Additive, alphaBoolBlend {blendOption=Linear}, boneToModify {boneName=dyn_skirt_01_02_L}}
   n37 = "Constraint - dyn_skirt_01_03_L (PR)"!(ComponentPose=@n36, Alpha=1, ConstraintWeights_0=1, ConstraintWeights_1:float=0) -> (Pose) {
    alphaBoolBlend {blendOption=Linear}
    boneToModify {boneName=dyn_skirt_01_03_L}
    constraintSetup [{targetBone {boneName=thigh_l}}, {transformType=Rotation, targetBone {boneName=pelvis}}]
    constraintWeights [1,1]
   }
   n38 = "Constraint - dyn_skirt_02_02_L (RR)"!(ComponentPose=@n37, Alpha=1, ConstraintWeights_0=0.8, ConstraintWeights_1:float=0.2) -> (Pose) {
    alphaBoolBlend {blendOption=Linear}
    boneToModify {boneName=dyn_skirt_02_02_L}
    constraintSetup [{transformType=Rotation, targetBone {boneName=thigh_l}}, {transformType=Rotation, targetBone {boneName=pelvis}}]
    constraintWeights [1,1]
   }
   n39 = "Constraint - dyn_skirt_02_03_L (PR)"!(ComponentPose=@n38, Alpha=1, ConstraintWeights_0=1, ConstraintWeights_1:float=0) -> (Pose) {
    alphaBoolBlend {blendOption=Linear}
    boneToModify {boneName=dyn_skirt_02_03_L}
    constraintSetup [{targetBone {boneName=thigh_l}}, {transformType=Rotation, targetBone {boneName=pelvis}}]
    constraintWeights [1,1]
   }
   n40 = "Constraint - dyn_skirt_03_02_L (RR)"!(ComponentPose=@n39, Alpha=1, ConstraintWeights_0=0.8, ConstraintWeights_1:float=0.2) -> (Pose) {
    alphaBoolBlend {blendOption=Linear}
    boneToModify {boneName=dyn_skirt_03_02_L}
    constraintSetup [{transformType=Rotation, targetBone {boneName=thigh_l}}, {transformType=Rotation, targetBone {boneName=pelvis}}]
    constraintWeights [1,1]
   }
   n41 = "Transform (Modify) Bone - Bone: dyn_skirt_03_02_L"(ComponentPose=@n40, Alpha=1, Translation="0,0,0", Rotation="0,0,0", Scale="1,1,1") {rotationMode=BMM_Additive, translationMode=BMM_Additive, alphaBoolBlend {blendOption=Linear}, boneToModify {boneName=dyn_skirt_03_02_L}}
   n42 = "Constraint - dyn_skirt_03_03_L (PR)"!(ComponentPose=@n41, Alpha=1, ConstraintWeights_0=1, ConstraintWeights_1:float=0) -> (Pose) {
    alphaBoolBlend {blendOption=Linear}
    boneToModify {boneName=dyn_skirt_03_03_L}
    constraintSetup [{targetBone {boneName=thigh_l}}, {transformType=Rotation, targetBone {boneName=pelvis}}]
    constraintWeights [1,1]
   }
   n43 = "Transform (Modify) Bone - Bone: dyn_skirt_03_03_L"(ComponentPose=@n42, Alpha=1, Translation="0,0,4", Rotation="0,0,0", Scale="1,1,1") {
    rotationMode=BMM_Additive
    translationMode=BMM_Additive
    alphaBoolBlend {blendOption=Linear}
    boneToModify {boneName=dyn_skirt_03_03_L}
    translation {z=4}
   }
   n44 = "Constraint - dyn_skirt_01_02_R (RR)"!(ComponentPose=@n43, Alpha=1, ConstraintWeights_0=0.8, ConstraintWeights_1:float=0.2) -> (Pose) {
    alphaBoolBlend {blendOption=Linear}
    boneToModify {boneName=dyn_skirt_01_02_R}
    constraintSetup [{transformType=Rotation, targetBone {boneName=thigh_r}}, {transformType=Rotation, targetBone {boneName=pelvis}}]
    constraintWeights [1,1]
   }
   n45 = "Transform (Modify) Bone - Bone: dyn_skirt_01_02_R"(ComponentPose=@n44, Alpha=1, Translation="0,0,0", Rotation="0,0,0", Scale="1,1,1") {rotationMode=BMM_Additive, translationMode=BMM_Additive, alphaBoolBlend {blendOption=Linear}, boneToModify {boneName=dyn_skirt_01_02_R}}
   n46 = "Constraint - dyn_skirt_01_03_R (PR)"!(ComponentPose=@n45, Alpha=1, ConstraintWeights_0=1, ConstraintWeights_1:float=0) -> (Pose) {
    alphaBoolBlend {blendOption=Linear}
    boneToModify {boneName=dyn_skirt_01_03_R}
    constraintSetup [{targetBone {boneName=thigh_r}}, {transformType=Rotation, targetBone {boneName=pelvis}}]
    constraintWeights [1,1]
   }
   n47 = "Constraint - dyn_skirt_02_02_R (RR)"!(ComponentPose=@n46, Alpha=1, ConstraintWeights_0=0.8, ConstraintWeights_1:float=0.2) -> (Pose) {
    alphaBoolBlend {blendOption=Linear}
    boneToModify {boneName=dyn_skirt_02_02_R}
    constraintSetup [{transformType=Rotation, targetBone {boneName=thigh_r}}, {transformType=Rotation, targetBone {boneName=pelvis}}]
    constraintWeights [1,1]
   }
   n48 = "Constraint - dyn_skirt_02_03_R (PR)"!(ComponentPose=@n47, Alpha=1, ConstraintWeights_0=1, ConstraintWeights_1:float=0) -> (Pose) {
    alphaBoolBlend {blendOption=Linear}
    boneToModify {boneName=dyn_skirt_02_03_R}
    constraintSetup [{targetBone {boneName=thigh_r}}, {transformType=Rotation, targetBone {boneName=pelvis}}]
    constraintWeights [1,1]
   }
   n49 = "Transform (Modify) Bone - Bone: dyn_skirt_03_01_R"(ComponentPose=@n48, Alpha=1, Translation="0,0,2", Rotation="0,0,0", Scale="1,1,1") {
    rotationMode=BMM_Additive
    translationMode=BMM_Additive
    alphaBoolBlend {blendOption=Linear}
    boneToModify {boneName=dyn_skirt_03_01_R}
    translation {z=2}
   }
   n50 = "Constraint - dyn_skirt_03_02_R (RR)"!(ComponentPose=@n49, Alpha=1, ConstraintWeights_0=0.8, ConstraintWeights_1:float=0.2) -> (Pose) {
    alphaBoolBlend {blendOption=Linear}
    boneToModify {boneName=dyn_skirt_03_02_R}
    constraintSetup [{transformType=Rotation, targetBone {boneName=thigh_r}}, {transformType=Rotation, targetBone {boneName=pelvis}}]
    constraintWeights [1,1]
   }
   n51 = "Transform (Modify) Bone - Bone: dyn_skirt_03_02_R"(ComponentPose=@n50, Alpha=1, Translation="0,0,0", Rotation="0,0,0", Scale="1,1,1") {rotationMode=BMM_Additive, translationMode=BMM_Additive, alphaBoolBlend {blendOption=Linear}, boneToModify {boneName=dyn_skirt_03_02_R}}
   n52 = "Constraint - dyn_skirt_03_03_R (PR)"!(ComponentPose=@n51, Alpha=1, ConstraintWeights_0=1, ConstraintWeights_1:float=0) -> (Pose) {
    alphaBoolBlend {blendOption=Linear}
    boneToModify {boneName=dyn_skirt_03_03_R}
    constraintSetup [{targetBone {boneName=thigh_r}}, {transformType=Rotation, targetBone {boneName=pelvis}}]
    constraintWeights [1,1]
   }
   n53 = "Transform (Modify) Bone - Bone: dyn_skirt_03_03_R"(ComponentPose=@n52, Alpha=1, Translation="0,0,4", Rotation="0,0,0", Scale="1,1,1") {
    rotationMode=BMM_Additive
    translationMode=BMM_Additive
    alphaBoolBlend {blendOption=Linear}
    boneToModify {boneName=dyn_skirt_03_03_R}
    translation {z=4}
   }
   n54 = "Constraint - dyn_skirt_F_01 (PP)"!(ComponentPose=@n53, Alpha=1, ConstraintWeights_0=0.5, ConstraintWeights_1:float=0.5) -> (Pose) {
    alphaBoolBlend {blendOption=Linear}
    boneToModify {boneName=dyn_skirt_F_01}
    constraintSetup [{targetBone {boneName=thigh_l}}, {targetBone {boneName=thigh_r}}]
    constraintWeights [1,1]
   }
   n55 = "Constraint - dyn_skirt_F_02 (PP)"!(ComponentPose=@n54, Alpha=1, ConstraintWeights_0=0.5, ConstraintWeights_1:float=0.5) -> (Pose) {
    alphaBoolBlend {blendOption=Linear}
    boneToModify {boneName=dyn_skirt_F_02}
    constraintSetup [{targetBone {boneName=thigh_l}}, {targetBone {boneName=thigh_r}}]
    constraintWeights [1,1]
   }
   n56 = "Constraint - dyn_skirt_F_03 (PP)"!(ComponentPose=@n55, Alpha=1, ConstraintWeights_0=0.5, ConstraintWeights_1:float=0.5) -> (Pose) {
    alphaBoolBlend {blendOption=Linear}
    boneToModify {boneName=dyn_skirt_F_03}
    constraintSetup [{targetBone {boneName=thigh_l}}, {targetBone {boneName=thigh_r}}]
    constraintWeights [1,1]
   }
   n57 = Blend Poses by bool(BlendPose_0=Component To Local(ComponentPose=@n56), BlendPose_1=Component To Local(ComponentPose=@n28), BlendTime_0=0.5, BlendTime_1=0.5) {blendType=Linear}
   n58 = "Constraint - dyn_strap_01_01 (PPP)"!(
    ComponentPose=Local To Component(LocalPose=Use cached pose 'Pre-Dynamics'())
    Alpha=1
    ConstraintWeights_0=0.78
    ConstraintWeights_1:float=0.06
    ConstraintWeights_2:float=0.16
   ) -> (Pose) {
    alphaBoolBlend {blendOption=Linear}
    boneToModify {boneName=dyn_strap_01_01}
    constraintSetup [{targetBone {boneName=pelvis}}, {targetBone {boneName=thigh_twist_01_l}}, {targetBone {boneName=deform_groin_l}}]
    constraintWeights [1,1,1]
   }
   n59 = "Constraint - dyn_strap_01_03 (PPPP)"!(
    ComponentPose=@n58
    Alpha=1
    ConstraintWeights_0=0.2
    ConstraintWeights_1:float=0.28
    ConstraintWeights_2:float=0.31
    ConstraintWeights_3:float=0.21
   ) -> (Pose) {
    alphaBoolBlend {blendOption=Linear}
    boneToModify {boneName=dyn_strap_01_03}
    constraintSetup [
     {targetBone {boneName=pelvis}}
     {targetBone {boneName=thigh_twist_01_l}}
     {targetBone {boneName=deform_groin_l}}
     {targetBone {boneName=dyn_skirt_01_01_L}}
    ]
    constraintWeights [1,1,1,1]
   }
   n60 = "Constraint - dyn_strap_01_02 (PP)"!(ComponentPose=@n59, Alpha=0, ConstraintWeights_0=0.5, ConstraintWeights_1:float=0.5) -> (Pose) {
    alphaBoolBlend {blendOption=Linear}
    boneToModify {boneName=dyn_strap_01_02}
    constraintSetup [{targetBone {boneName=dyn_strap_01_01}}, {targetBone {boneName=dyn_strap_01_03}}]
    constraintWeights [1,1]
   }
   n61 = "Constraint - dyn_strap_02_01 (PPP)"!(ComponentPose=@n60, Alpha=1, ConstraintWeights_0=0.8, ConstraintWeights_1:float=0.12, ConstraintWeights_2:float=0.08) -> (Pose) {
    alphaBoolBlend {blendOption=Linear}
    boneToModify {boneName=dyn_strap_02_01}
    constraintSetup [{targetBone {boneName=pelvis}}, {targetBone {boneName=thigh_twist_01_l}}, {targetBone {boneName=deform_groin_l}}]
    constraintWeights [1,1,1]
   }
   n62 = "Constraint - dyn_strap_02_02 (PP)"!(ComponentPose=@n61, Alpha=1, ConstraintWeights_0=0.5, ConstraintWeights_1:float=0.5) -> (Pose) {
    alphaBoolBlend {blendOption=Linear}
    boneToModify {boneName=dyn_strap_02_02}
    constraintSetup [{targetBone {boneName=dyn_strap_02_01}}, {targetBone {boneName=dyn_skirt_02_01_L}}]
    constraintWeights [1,1]
   }
   n63 = "Constraint - dyn_strap_02_03 (PPPPP)"!(
    ComponentPose=@n62
    Alpha=1
    ConstraintWeights_0=0.26
    ConstraintWeights_1:float=0.18
    ConstraintWeights_2:float=0.4
    ConstraintWeights_3:float=0.11
    ConstraintWeights_4:float=0.05
   ) -> (Pose) {
    alphaBoolBlend {blendOption=Linear}
    boneToModify {boneName=dyn_strap_02_03}
    constraintSetup [
     {targetBone {boneName=thigh_twist_01_l}}
     {targetBone {boneName=dyn_skirt_02_01_L}}
     {targetBone {boneName=dyn_skirt_02_02_L}}
     {targetBone {boneName=dyn_skirt_02_03_L}}
     {targetBone {boneName=dyn_skirt_03_01_L}}
    ]
    constraintWeights [1,1,1,1,1]
   }
   n64 = "Constraint - dyn_skirt_02_mstr_R (P)"(ComponentPose=@n63, Alpha=1, ConstraintWeights_0=1) {
    alphaBoolBlend {blendOption=Linear}
    boneToModify {boneName=dyn_skirt_02_mstr_R}
    constraintSetup [{targetBone {boneName=thigh_r}}]
    constraintWeights [1]
   }
   n65 = "Transform (Modify) Bone - Bone: dyn_skirt_03_mstr_R"!(ComponentPose=@n64, Alpha=1, Translation="0,0,0", Rotation="0,0,20") -> (Pose) {
    rotationMode=BMM_Additive
    rotationSpace=BCS_BoneSpace
    translationMode=BMM_Additive
    translationSpace=BCS_BoneSpace
    alphaBoolBlend {blendOption=Linear}
    alphaScaleBiasClamp {bInterpResult=true, interpSpeedDecreasing=15, interpSpeedIncreasing=17}
    boneToModify {boneName=dyn_skirt_03_mstr_R}
    rotation {roll=20}
   }
   n66 = "Transform (Modify) Bone - Bone: dyn_skirt_02_mstr_R"!(ComponentPose=@n65, Alpha=1, Translation="0,0,0", Rotation="10,0,5") -> (Pose) {
    rotationMode=BMM_Additive
    rotationSpace=BCS_BoneSpace
    translationMode=BMM_Additive
    translationSpace=BCS_BoneSpace
    alphaBoolBlend {blendOption=Linear}
    alphaScaleBiasClamp {bInterpResult=true, interpSpeedDecreasing=15, interpSpeedIncreasing=17}
    boneToModify {boneName=dyn_skirt_02_mstr_R}
    rotation {pitch=10, roll=5}
   }
   n67 = "Transform (Modify) Bone - Bone: dyn_skirt_03_mstr_L"!(ComponentPose=@n66, Alpha=1, Translation="0,0,0", Rotation="20,0,5") -> (Pose) {
    rotationMode=BMM_Additive
    rotationSpace=BCS_BoneSpace
    translationMode=BMM_Additive
    translationSpace=BCS_BoneSpace
    alphaBoolBlend {blendOption=Linear}
    alphaScaleBiasClamp {bInterpResult=true, interpSpeedDecreasing=15, interpSpeedIncreasing=17}
    boneToModify {boneName=dyn_skirt_03_mstr_L}
    rotation {pitch=20, roll=5}
   }
   n68 = "Transform (Modify) Bone - Bone: dyn_skirt_02_mstr_L"!(ComponentPose=@n67, Alpha=1, Translation="0,0,0", Rotation="5.000,-0.000,-0.000") -> (Pose) {
    rotationMode=BMM_Additive
    rotationSpace=BCS_BoneSpace
    translationMode=BMM_Additive
    translationSpace=BCS_BoneSpace
    alphaBoolBlend {blendOption=Linear}
    alphaScaleBiasClamp {bInterpResult=true, interpSpeedDecreasing=15, interpSpeedIncreasing=17}
    boneToModify {boneName=dyn_skirt_02_mstr_L}
    rotation=<rot 5 0 -0.0001>
   }
   n69 = "Transform (Modify) Bone - Bone: dyn_skirt_03_mstr_R"(ComponentPose=@n68, Alpha=1, Translation="0,0,0", Rotation="0,0,15") {
    rotationMode=BMM_Additive
    rotationSpace=BCS_BoneSpace
    translationMode=BMM_Additive
    translationSpace=BCS_BoneSpace
    alphaBoolBlend {blendOption=Linear}
    alphaScaleBiasClamp {bInterpResult=true, interpSpeedDecreasing=15, interpSpeedIncreasing=17}
    boneToModify {boneName=dyn_skirt_03_mstr_R}
    rotation {roll=15}
   }
   n70 = Component To Local(ComponentPose=@n69)
   n71 = "Constraint - dyn_skirt_02_01_R (P)"(
    ComponentPose=Local To Component(LocalPose=Blend Poses by bool(BlendPose_0=@n57, BlendPose_1=@n70, BlendTime_0=0.5, BlendTime_1=0.5) {blendType=Linear})
    Alpha=1
    ConstraintWeights_0=1
   ) {
    alphaBoolBlend {blendOption=Linear}
    alphaScaleBiasClamp {bInterpResult=true, interpSpeedDecreasing=100, interpSpeedIncreasing=1}
    boneToModify {boneName=dyn_skirt_02_01_R}
    constraintSetup [{targetBone {boneName=thigh_r}}]
    constraintWeights [1]
   }
   n72 = "Constraint - dyn_skirt_02_02_R (P)"(ComponentPose=@n71, Alpha=1, ConstraintWeights_0=1) {
    alphaBoolBlend {blendOption=Linear}
    alphaScaleBiasClamp {bInterpResult=true, interpSpeedDecreasing=100, interpSpeedIncreasing=1}
    boneToModify {boneName=dyn_skirt_02_02_R}
    constraintSetup [{targetBone {boneName=thigh_r}}]
    constraintWeights [1]
   }
   n73 = "Constraint - dyn_skirt_02_03_R (P)"(ComponentPose=@n72, Alpha=1, ConstraintWeights_0=1) {
    alphaBoolBlend {blendOption=Linear}
    alphaScaleBiasClamp {bInterpResult=true, interpSpeedDecreasing=100, interpSpeedIncreasing=1}
    boneToModify {boneName=dyn_skirt_02_03_R}
    constraintSetup [{targetBone {boneName=thigh_r}}]
    constraintWeights [1]
   }
   n74 = "Constraint - dyn_skirt_01_01_R (P)"(ComponentPose=@n73, Alpha=1, ConstraintWeights_0=1) {
    alphaBoolBlend {blendOption=Linear}
    alphaScaleBiasClamp {bInterpResult=true, interpSpeedDecreasing=100, interpSpeedIncreasing=1}
    boneToModify {boneName=dyn_skirt_01_01_R}
    constraintSetup [{targetBone {boneName=thigh_r}}]
    constraintWeights [1]
   }
   n75 = "Constraint - dyn_skirt_01_02_R (P)"(ComponentPose=@n74, Alpha=1, ConstraintWeights_0=1) {
    alphaBoolBlend {blendOption=Linear}
    alphaScaleBiasClamp {bInterpResult=true, interpSpeedDecreasing=100, interpSpeedIncreasing=1}
    boneToModify {boneName=dyn_skirt_01_02_R}
    constraintSetup [{targetBone {boneName=thigh_r}}]
    constraintWeights [1]
   }
   n76 = "Constraint - dyn_skirt_01_03_R (P)"(ComponentPose=@n75, Alpha=1, ConstraintWeights_0=1) {
    alphaBoolBlend {blendOption=Linear}
    alphaScaleBiasClamp {bInterpResult=true, interpSpeedDecreasing=100, interpSpeedIncreasing=1}
    boneToModify {boneName=dyn_skirt_01_03_R}
    constraintSetup [{targetBone {boneName=thigh_r}}]
    constraintWeights [1]
   }
   n77 = "Constraint - dyn_skirt_03_01_R (P)"(ComponentPose=@n76, Alpha=1, ConstraintWeights_0=1) {
    alphaBoolBlend {blendOption=Linear}
    alphaScaleBiasClamp {bInterpResult=true, interpSpeedDecreasing=100, interpSpeedIncreasing=1}
    boneToModify {boneName=dyn_skirt_03_01_R}
    constraintSetup [{targetBone {boneName=thigh_r}}]
    constraintWeights [1]
   }
   n78 = "Transform (Modify) Bone - Bone: dyn_skirt_03_01_R"(ComponentPose=@n77, Alpha=1, Translation="0,-3,3", Rotation="-0.000,0.000,4.453", Scale="1,1,1") {
    rotationMode=BMM_Additive
    rotationSpace=BCS_BoneSpace
    translationMode=BMM_Additive
    translationSpace=BCS_BoneSpace
    alphaBoolBlend {blendOption=Linear}
    alphaScaleBiasClamp {bInterpResult=true, interpSpeedIncreasing=1}
    boneToModify {boneName=dyn_skirt_03_01_R}
    rotation {roll=4.453}
    translation {y=-3, z=3}
   }
   n79 = "Transform (Modify) Bone - Bone: dyn_skirt_03_01_R"(ComponentPose=@n78, Alpha=1, Translation="-18.441,-9.321,6.018", Rotation="0,0,0", Scale="1,1,1") {
    rotationMode=BMM_Additive
    rotationSpace=BCS_BoneSpace
    scaleMode=BMM_Additive
    scaleSpace=BCS_BoneSpace
    translationMode=BMM_Additive
    translationSpace=BCS_BoneSpace
    alphaBoolBlend {blendOption=Linear}
    alphaScaleBiasClamp {bInterpResult=true, interpSpeedIncreasing=1}
    boneToModify {boneName=dyn_skirt_03_01_R}
    translation=<-18.441 -9.321 6.018>
   }
   n80 = "Constraint - dyn_skirt_03_02_R (P)"(ComponentPose=@n79, Alpha=1, ConstraintWeights_0=1) {
    alphaBoolBlend {blendOption=Linear}
    alphaScaleBiasClamp {bInterpResult=true, interpSpeedDecreasing=100, interpSpeedIncreasing=1}
    boneToModify {boneName=dyn_skirt_03_02_R}
    constraintSetup [{targetBone {boneName=thigh_r}}]
    constraintWeights [1]
   }
   n81 = "Transform (Modify) Bone - Bone: dyn_skirt_03_02_R"(ComponentPose=@n80, Alpha=1, Translation="0,0,0", Rotation="-35,0,0", Scale="1,1,1") {
    rotationMode=BMM_Additive
    rotationSpace=BCS_BoneSpace
    translationMode=BMM_Additive
    translationSpace=BCS_BoneSpace
    alphaBoolBlend {blendOption=Linear}
    alphaScaleBiasClamp {bInterpResult=true, interpSpeedIncreasing=1}
    boneToModify {boneName=dyn_skirt_03_02_R}
    rotation {pitch=-35}
   }
   n82 = "Transform (Modify) Bone - Bone: dyn_skirt_03_02_R"(ComponentPose=@n81, Alpha=1, Translation="-19.644,-10.738,6.069", Rotation="0,0,0", Scale="1,1,1") {
    rotationMode=BMM_Additive
    rotationSpace=BCS_BoneSpace
    scaleMode=BMM_Additive
    scaleSpace=BCS_BoneSpace
    translationMode=BMM_Additive
    translationSpace=BCS_BoneSpace
    alphaBoolBlend {blendOption=Linear}
    alphaScaleBiasClamp {bInterpResult=true, interpSpeedIncreasing=1}
    boneToModify {boneName=dyn_skirt_03_02_R}
    translation=<-19.644 -10.738 6.069>
   }
   n83 = "Constraint - dyn_skirt_03_03_R (P)"(ComponentPose=@n82, Alpha=1, ConstraintWeights_0=1) {
    alphaBoolBlend {blendOption=Linear}
    alphaScaleBiasClamp {bInterpResult=true, interpSpeedDecreasing=100, interpSpeedIncreasing=1}
    boneToModify {boneName=dyn_skirt_03_03_R}
    constraintSetup [{targetBone {boneName=thigh_r}}]
    constraintWeights [1]
   }
   n84 = "Transform (Modify) Bone - Bone: dyn_skirt_03_03_R"(ComponentPose=@n83, Alpha=1, Translation="-2.35,0,-4.335", Rotation="-40,0,0", Scale="1,1,1") {
    rotationMode=BMM_Additive
    rotationSpace=BCS_BoneSpace
    translationMode=BMM_Additive
    translationSpace=BCS_BoneSpace
    alphaBoolBlend {blendOption=Linear}
    alphaScaleBiasClamp {bInterpResult=true, interpSpeedIncreasing=1}
    boneToModify {boneName=dyn_skirt_03_03_R}
    rotation {pitch=-40}
    translation {x=-2.35, z=-4.335}
   }
   n85 = "Transform (Modify) Bone - Bone: dyn_skirt_03_03_R"(ComponentPose=@n84, Alpha=1, Translation="-15.942,-12.003,6.412", Rotation="0,0,0", Scale="1,1,1") {
    rotationMode=BMM_Additive
    rotationSpace=BCS_BoneSpace
    scaleMode=BMM_Additive
    scaleSpace=BCS_BoneSpace
    translationMode=BMM_Additive
    translationSpace=BCS_BoneSpace
    alphaBoolBlend {blendOption=Linear}
    alphaScaleBiasClamp {bInterpResult=true, interpSpeedIncreasing=1}
    boneToModify {boneName=dyn_skirt_03_03_R}
    translation=<-15.942 -12.003 6.412>
   }
   n86 = "Constraint - dyn_skirt_03_01_L (P)"(ComponentPose=@n85, Alpha=1, ConstraintWeights_0=1) {
    alphaBoolBlend {blendOption=Linear}
    alphaScaleBiasClamp {bInterpResult=true, interpSpeedDecreasing=100, interpSpeedIncreasing=1}
    boneToModify {boneName=dyn_skirt_03_01_L}
    constraintSetup [{targetBone {boneName=thigh_l}}]
    constraintWeights [1]
   }
   n87 = "Transform (Modify) Bone - Bone: dyn_skirt_03_01_L"(ComponentPose=@n86, Alpha=1, Translation="-3,-3,4", Rotation="0,0,0", Scale="1,1,1") {
    rotationMode=BMM_Additive
    rotationSpace=BCS_BoneSpace
    translationMode=BMM_Additive
    translationSpace=BCS_BoneSpace
    alphaBoolBlend {blendOption=Linear}
    alphaScaleBiasClamp {bInterpResult=true, interpSpeedIncreasing=1}
    boneToModify {boneName=dyn_skirt_03_01_L}
    translation=<-3 -3 4>
   }
   n88 = "Transform (Modify) Bone - Bone: dyn_skirt_03_01_L"(ComponentPose=@n87, Alpha=1, Translation="-7.212,-0.178,11.324", Rotation="0,0,0", Scale="1,1,1") {
    rotationMode=BMM_Additive
    rotationSpace=BCS_BoneSpace
    translationMode=BMM_Additive
    translationSpace=BCS_BoneSpace
    alphaBoolBlend {blendOption=Linear}
    alphaScaleBiasClamp {bInterpResult=true, interpSpeedIncreasing=1}
    boneToModify {boneName=dyn_skirt_03_01_L}
    translation=<-7.212 -0.178 11.324>
   }
   n89 = "Constraint - dyn_skirt_03_02_L (P)"(ComponentPose=@n88, Alpha=1, ConstraintWeights_0=1) {
    alphaBoolBlend {blendOption=Linear}
    alphaScaleBiasClamp {bInterpResult=true, interpSpeedDecreasing=100, interpSpeedIncreasing=1}
    boneToModify {boneName=dyn_skirt_03_02_L}
    constraintSetup [{targetBone {boneName=thigh_l}}]
    constraintWeights [1]
   }
   n90 = "Transform (Modify) Bone - Bone: dyn_skirt_03_02_L"(ComponentPose=@n89, Alpha=1, Translation="-0.000,0.264,6.807", Rotation="40.731,0.000,0.000", Scale="1,1,1") {
    rotationMode=BMM_Additive
    rotationSpace=BCS_BoneSpace
    translationMode=BMM_Additive
    translationSpace=BCS_BoneSpace
    alphaBoolBlend {blendOption=Linear}
    alphaScaleBiasClamp {bInterpResult=true, interpSpeedDecreasing=5, interpSpeedIncreasing=5}
    boneToModify {boneName=dyn_skirt_03_02_L}
    rotation=<rot 40.731 0 0>
    translation {y=0.264, z=6.807}
   }
   n91 = "Constraint - dyn_skirt_03_03_L (P)"(ComponentPose=@n90, Alpha=1, ConstraintWeights_0=1) {
    alphaBoolBlend {blendOption=Linear}
    alphaScaleBiasClamp {bInterpResult=true, interpSpeedDecreasing=100, interpSpeedIncreasing=1}
    boneToModify {boneName=dyn_skirt_03_03_L}
    constraintSetup [{targetBone {boneName=thigh_l}}]
    constraintWeights [1]
   }
   n92 = "Constraint - dyn_strap_02_02 (P)"(ComponentPose=@n91, Alpha=1, ConstraintWeights_0=1) {
    alphaBoolBlend {blendOption=Linear}
    boneToModify {boneName=dyn_strap_02_02}
    constraintSetup [{targetBone {boneName=dyn_skirt_02_01_L}}]
    constraintWeights [1]
   }
   n93 = "Transform (Modify) Bone - Bone: dyn_strap_02_01"(ComponentPose=@n92, Alpha=1, Translation="0,0,2", Rotation="0,0,0", Scale="1,1,1") {
    rotationMode=BMM_Additive
    rotationSpace=BCS_BoneSpace
    translationMode=BMM_Additive
    translationSpace=BCS_BoneSpace
    alphaBoolBlend {blendOption=Linear}
    alphaScaleBiasClamp {bInterpResult=true, interpSpeedIncreasing=1}
    boneToModify {boneName=dyn_strap_02_01}
    translation {z=2}
   }
   n94 = "Transform (Modify) Bone - Bone: dyn_skirt_03_mstr_R"!(ComponentPose=@n93, Alpha=1, Rotation="10,0,15") -> (Pose) {
    rotationMode=BMM_Additive
    rotationSpace=BCS_BoneSpace
    translationSpace=BCS_BoneSpace
    alphaBoolBlend {blendOption=Linear}
    alphaScaleBiasClamp {bInterpResult=true, interpSpeedIncreasing=1}
    boneToModify {boneName=dyn_skirt_03_mstr_R}
    rotation {pitch=10, roll=15}
   }
   n95 = "Transform (Modify) Bone - Bone: dyn_skirt_02_mstr_R"!(ComponentPose=@n94, Alpha=1, Rotation="10.000,-0.000,-0.000") -> (Pose) {
    rotationMode=BMM_Additive
    rotationSpace=BCS_BoneSpace
    translationSpace=BCS_BoneSpace
    alphaBoolBlend {blendOption=Linear}
    alphaScaleBiasClamp {bInterpResult=true, interpSpeedIncreasing=1}
    boneToModify {boneName=dyn_skirt_02_mstr_R}
    rotation=<rot 10 0 0>
   }
   Output Pose(Result=Component To Local(ComponentPose=@n95)) {layerGroup=None}
   n96 = Control Rig!(
    Source=Copy Pose From Mesh() {bUseAttachedParent=true}
    Alpha
    IsCrouchCtrl:bool=false
    SpecificEmoteCtrl:bool=false
    SkyDiveCtrl:bool=false
    SpecificAnimCtrl:bool=false
    RbanOffCtrl:bool=false
    Melee:bool=Get bIsPlayingMeleeAnim()
   ) -> (Pose) {
    alphaBoolBlend {blendOption=Linear}
    controlRigAssetReference {blueprintRigClass=&/F_MED_LiftingAura_Wave_CtrlRig.F_MED_LiftingAura_Wave_CtrlRig_C}
   }
   Save cached pose 'Pre-Dynamics'(Pose=@n96) {cacheName=Pre-Dynamics}
   n97 = "Transform (Modify) Bone - Bone: dyn_strap_01_01"(ComponentPose=Local To Component(LocalPose=Use cached pose 'Pre-Dynamics'()), Alpha=1, Translation="0,0,0", Rotation="65,0,0", Scale="1,1,1") {
    rotationMode=BMM_Additive
    rotationSpace=BCS_BoneSpace
    translationMode=BMM_Additive
    translationSpace=BCS_BoneSpace
    alphaBoolBlend {blendOption=Linear}
    alphaScaleBiasClamp {bInterpResult=true, interpSpeedIncreasing=2}
    boneToModify {boneName=dyn_strap_01_01}
    rotation {pitch=65}
   }
   n98 = "Transform (Modify) Bone - Bone: dyn_strap_02_01"(ComponentPose=@n97, Alpha=1, Translation="0,0,0", Rotation="65,0,0", Scale="1,1,1") {
    rotationMode=BMM_Additive
    rotationSpace=BCS_BoneSpace
    translationMode=BMM_Additive
    translationSpace=BCS_BoneSpace
    alphaBoolBlend {blendOption=Linear}
    alphaScaleBiasClamp {bInterpResult=true, interpSpeedIncreasing=2}
    boneToModify {boneName=dyn_strap_02_01}
    rotation {pitch=65}
   }
   Save cached pose 'PreSim'(Pose=Component To Local(ComponentPose=@n98)) {cacheName=PreSim}
  ]}
 EventGraph {nodes [
   Event BlueprintUpdateAnimation()
   > n1 = Sequence!(execute) -> (then_0, then_1, then_2:Exec, then_3:Exec, then_4:Exec)
   > n2 = Gravity Override State Function DEPRECATED(CharacterPartAnimInstance=Self-Reference(), GameTimeInSeconds=GetGameTimeInSeconds(), GravityOverrideParams=Get Gravity Override Params())
   > Set Gravity Override(Gravity Override=@n2)
   n3 = Get bIsCrouching()
   Set Crouching(execute=@n1.then_1, Crouching:double=SelectFloat(A=1, B=0, bPickA=@n3)) -> (Output_Get:double)
   > Set Crouching2(Crouching2:double=SelectFloat(A=0.5, B=0, bPickA=@n3)) -> (Output_Get:double)
   n4 = GetAnimInstance(self=GetOwnerSkeletalMeshComponent())
   Is Valid(exec=@n1.then_2, InputObject=@n4)
   > Set SpecificEmote(SpecificEmote:bool=Montage_IsPlaying(self=@n4, Montage=/BRCosmetics/Animation/Game/MainPlayer/Emotes/LiftingAura/Emote_LiftingAura_CMF_M)) -> (Output_Get:bool)
   n5 = GetDisplayName(Object=Get CurrentWeapon() -> (CurrentWeapon:Fort Weapon Object Reference))
   > Set SpecificAnim(SpecificAnim:bool=OR Boolean!(A="Equal Exactly (String)"(A=@n5, B=B_Athena_SunRose_Zeus_Lightning0), B="Equal Exactly (String)"(A=@n5, B=B_Weapon_AnnoyedMop0), C:bool) -> (ReturnValue)) -> (Output_Get:bool)
   n6 = BreakTransform(InTransform=GetDeltaTransformFromRefPose(self=GetOwnerSkeletalMeshComponent(), BoneName=thigh_l, BaseName=pelvis))
   Set ThighLRotation(execute=@n1.then_3, ThighLRotation:double=MapRangeClamped(Value=BreakRotator(InRot=@n6.Rotation).Yaw, InRangeA=-50, InRangeB=-90)) -> (Output_Get:double)
   n7 = Get bIsSkyDiving()
   n8 = "Not Equal (Boolean)"(A=@n7, B=true) -> (ReturnValue=~)
   n9 = CheckForBackbling(self=Self-Reference())
   n10 = "Not Equal (Boolean)"(A=@n9.bIsBackblingACape, B=true) -> (ReturnValue=~)
   Set SkyDiveNot(execute=@n1.then_4, SkyDiveNot:double=SelectFloat(A=1, B=0, bPickA=AND Boolean(A=@n8, B=@n10))) -> (Output_Get:double)
   n11 = Get bIsSkyDiving()
   > Set SkyDiveOn(SkyDiveOn:double=SelectFloat(A=1, B=0, bPickA=AND Boolean(A=@n11, B=@n10))) -> (Output_Get:double)
   n12 = Get bIsSkyDiving()
   > Set CapeOn(CapeOn:bool=AND Boolean(A="Not Equal (Boolean)"(A=@n12, B=true) -> (ReturnValue=~), B=@n9.bIsBackblingACape)) -> (Output_Get:bool)
   > Set FrontEnd(FrontEnd:bool=Get bIsInFrontEnd()) -> (Output_Get:bool)
   n13 = Get bRigidBodyEnabled()
   > Set RbanOff(RbanOff:bool="Not Equal (Boolean)"(A=@n13) -> (ReturnValue=~)) -> (Output_Get:bool)
   n14 = Event OnBeginFalling()
   Set jumping(execute=[@n14.then, @n16.then], jumping:double=1) -> (Output_Get:double)
   n15 = Event OnEndFalling()
   Set jumping(execute=[@n15.then, @n17.then], jumping=0)
   n16 = Event OnBeginJumping()
   n17 = Event OnEndJumping()
   NOT Boolean(A=@n13)
   OR Boolean!(A=@n7, B=Get bIsSkydiveFloating()) -> (ReturnValue)
  ]}
]
variables [
 CapeOn {type=boolean}
 Crouching {type=number}
 Crouching2 {type=number}
 FrontEnd {type=boolean}
 Gravity Override {type=Vector}
 Gravity Override Params {type=GravityOverrideParamsStruct}
 jumping {type=number}
 RbanOff {type=boolean}
 SkyDiveNot {type=number}
 SkyDiveOn {type=number}
 SpecificAnim {type=boolean}
 SpecificEmote {type=boolean}
 ThighLRotation {type=number}
]
