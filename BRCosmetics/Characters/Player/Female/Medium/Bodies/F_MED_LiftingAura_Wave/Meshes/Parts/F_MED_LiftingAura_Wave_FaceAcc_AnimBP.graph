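# Post-attach Anim Blueprint for the LiftingAura Wave female face-accessory mesh, adding secondary physics motion to its hair bones.
# It copies the attached parent's pose, then runs a serial chain of single-bone Anim Dynamics nodes over the dyn_hair bones.
# Each uses root-relative simulation, angular springs, damped component motion and per-bone angular limits, tightest on the back hair.
# Jump and fall events toggle a `jumping` value, which the AnimGraph never reads.
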AnimBlueprint /BRCosmetics/Characters/Player/Female/Medium/Bodies/F_MED_LiftingAura_Wave/Meshes/Parts/F_MED_LiftingAura_Wave_FaceAcc_AnimBP
target_skeleton=&/F_MED_LiftingAura_Wave_FaceAcc_Skeleton
graphs [
 AnimGraph {nodes [
   n1 = Anim Dynamics - Bone: dyn_hair_01_01(ComponentPose=Local To Component(LocalPose=Copy Pose From Mesh() {bUseAttachedParent=true}), Alpha=1, ExternalForce="0,0,0") {
    angularSpringConstant=100
    bAngularSpring=true
    bUseGravityOverride=true
    simulationSpace=RootRelative
    alphaBoolBlend {blendOption=Linear}
    boundBone {boneName=dyn_hair_01_01}
    componentAppliedLinearAccClamp=<3000 3000 3000>
    componentLinearAccScale=<0.2 0.2 0.2>
    componentLinearVelScale=<0.2 0.2 0.2>
    gravityOverride {z=-980}
    physicsBodyDefinitions [{
      boundBone {boneName=dyn_hair_01_01}
      boxExtents=<5 25 25>
      constraintSetup {angularLimitsMax=<1 0 1>, angularLimitsMin=<-2 -20 -4>, angularTarget=<1 0 0>}
      localJointOffset=<3 0 0>
     }]
    relativeSpaceBone {boneName=spine_05}
   }
   n2 = Anim Dynamics - Bone: dyn_hair_01_02(ComponentPose=@n1, Alpha=1, ExternalForce="0,0,0") {
    angularSpringConstant=100
    bAngularSpring=true
    bUseGravityOverride=true
    simulationSpace=RootRelative
    alphaBoolBlend {blendOption=Linear}
    boundBone {boneName=dyn_hair_01_02}
    componentAppliedLinearAccClamp=<3000 3000 3000>
    componentLinearAccScale=<0.2 0.2 0.2>
    componentLinearVelScale=<0.2 0.2 0.2>
    gravityOverride {z=-980}
    physicsBodyDefinitions [{
      boundBone {boneName=dyn_hair_01_02}
      boxExtents=<5 25 25>
      constraintSetup {angularLimitsMax=<1 0 1>, angularLimitsMin=<-2 -20 -4>, angularTarget=<1 0 0>}
      localJointOffset=<3 0 0>
     }]
    relativeSpaceBone {boneName=spine_05}
   }
   n3 = Anim Dynamics - Bone: dyn_hair_01_03(ComponentPose=@n2, Alpha=1, ExternalForce="0,0,0") {
    angularSpringConstant=100
    bAngularSpring=true
    bUseGravityOverride=true
    simulationSpace=RootRelative
    alphaBoolBlend {blendOption=Linear}
    boundBone {boneName=dyn_hair_01_03}
    componentAppliedLinearAccClamp=<3000 3000 3000>
    componentLinearAccScale=<0.2 0.2 0.2>
    componentLinearVelScale=<0.2 0.2 0.2>
    gravityOverride {z=-980}
    physicsBodyDefinitions [{
      boundBone {boneName=dyn_hair_01_03}
      boxExtents=<5 25 25>
      constraintSetup {angularLimitsMax=<1 0 1>, angularLimitsMin=<-2 -20 -4>, angularTarget=<1 0 0>}
      localJointOffset=<3 0 0>
     }]
    relativeSpaceBone {boneName=spine_05}
   }
   n4 = Anim Dynamics - Bone: dyn_hair_02_01(ComponentPose=@n3, Alpha=1, ExternalForce="0,0,0") {
    angularSpringConstant=100
    bAngularSpring=true
    bUseGravityOverride=true
    simulationSpace=RootRelative
    alphaBoolBlend {blendOption=Linear}
    boundBone {boneName=dyn_hair_02_01}
    componentAppliedLinearAccClamp=<3000 3000 3000>
    componentLinearAccScale=<0.2 0.2 0.2>
    componentLinearVelScale=<0.2 0.2 0.2>
    gravityOverride {z=-980}
    physicsBodyDefinitions [{
      boundBone {boneName=dyn_hair_02_01}
      boxExtents=<5 25 25>
      constraintSetup {angularLimitsMax=<2 1 4>, angularLimitsMin=<-4 -19 -4>, angularTarget=<1 0 0>}
      localJointOffset=<3 0 0>
     }]
    relativeSpaceBone {boneName=spine_05}
   }
   n5 = Anim Dynamics - Bone: dyn_hair_03_01(ComponentPose=@n4, Alpha=1, ExternalForce="0,0,0") {
    angularSpringConstant=100
    bAngularSpring=true
    bUseGravityOverride=true
    simulationSpace=RootRelative
    alphaBoolBlend {blendOption=Linear}
    boundBone {boneName=dyn_hair_03_01}
    componentAppliedLinearAccClamp=<3000 3000 3000>
    componentLinearAccScale=<0.2 0.2 0.2>
    componentLinearVelScale=<0.2 0.2 0.2>
    gravityOverride {z=-980}
    physicsBodyDefinitions [{
      boundBone {boneName=dyn_hair_03_01}
      boxExtents=<5 25 25>
      constraintSetup {angularLimitsMax=<4 3 4>, angularLimitsMin=<-3 -18 -4>, angularTarget=<1 0 0>}
      localJointOffset=<3 0 0>
     }]
    relativeSpaceBone {boneName=spine_05}
   }
   n6 = Anim Dynamics - Bone: dyn_hair_04_01(ComponentPose=@n5, Alpha=1, ExternalForce="0,0,0") {
    angularSpringConstant=100
    bAngularSpring=true
    bUseGravityOverride=true
    simulationSpace=RootRelative
    alphaBoolBlend {blendOption=Linear}
    boundBone {boneName=dyn_hair_04_01}
    componentAppliedLinearAccClamp=<3000 3000 3000>
    componentLinearAccScale=<0.2 0.2 0.2>
    componentLinearVelScale=<0.2 0.2 0.2>
    gravityOverride {z=-980}
    physicsBodyDefinitions [{
      boundBone {boneName=dyn_hair_04_01}
      boxExtents=<5 25 25>
      constraintSetup {angularLimitsMax=<6 5 3>, angularLimitsMin=<-6 -5 -3>, angularTarget=<1 0 0>}
      localJointOffset=<3 0 0>
     }]
    relativeSpaceBone {boneName=spine_05}
   }
   n7 = Anim Dynamics - Bone: dyn_hair_04_02(ComponentPose=@n6, Alpha=1, ExternalForce="0,0,0") {
    angularSpringConstant=100
    bAngularSpring=true
    bUseGravityOverride=true
    simulationSpace=RootRelative
    alphaBoolBlend {blendOption=Linear}
    boundBone {boneName=dyn_hair_04_02}
    componentAppliedLinearAccClamp=<3000 3000 3000>
    componentLinearAccScale=<0.2 0.2 0.2>
    componentLinearVelScale=<0.2 0.2 0.2>
    gravityOverride {z=-980}
    physicsBodyDefinitions [{
      boundBone {boneName=dyn_hair_04_02}
      boxExtents=<5 25 25>
      constraintSetup {angularLimitsMax=<6 5 3>, angularLimitsMin=<-6 -5 -3>, angularTarget=<1 0 0>}
      localJointOffset=<3 0 0>
     }]
    relativeSpaceBone {boneName=spine_05}
   }
   n8 = Anim Dynamics - Bone: dyn_hair_05_01(ComponentPose=@n7, Alpha=1, ExternalForce="0,0,0") {
    angularSpringConstant=100
    bAngularSpring=true
    bUseGravityOverride=true
    simulationSpace=RootRelative
    alphaBoolBlend {blendOption=Linear}
    boundBone {boneName=dyn_hair_05_01}
    componentAppliedLinearAccClamp=<3000 3000 3000>
    componentLinearAccScale=<0.2 0.2 0.2>
    componentLinearVelScale=<0.2 0.2 0.2>
    gravityOverride {z=-980}
    physicsBodyDefinitions [{
      boundBone {boneName=dyn_hair_05_01}
      boxExtents=<5 25 25>
      constraintSetup {angularLimitsMax=<3 5 3>, angularLimitsMin=<-3 -5 -3>, angularTarget=<1 0 0>}
      localJointOffset=<3 0 0>
     }]
    relativeSpaceBone {boneName=spine_05}
   }
   n9 = Anim Dynamics - Bone: dyn_hair_05_02(ComponentPose=@n8, Alpha=1, ExternalForce="0,0,0") {
    angularSpringConstant=100
    bAngularSpring=true
    bUseGravityOverride=true
    simulationSpace=RootRelative
    alphaBoolBlend {blendOption=Linear}
    boundBone {boneName=dyn_hair_05_02}
    componentAppliedLinearAccClamp=<3000 3000 3000>
    componentLinearAccScale=<0.2 0.2 0.2>
    componentLinearVelScale=<0.2 0.2 0.2>
    gravityOverride {z=-980}
    physicsBodyDefinitions [{
      boundBone {boneName=dyn_hair_05_02}
      boxExtents=<5 25 25>
      constraintSetup {angularLimitsMax=<3 5 3>, angularLimitsMin=<-3 -5 -3>, angularTarget=<1 0 0>}
      localJointOffset=<3 0 0>
     }]
    relativeSpaceBone {boneName=spine_05}
   }
   n10 = Anim Dynamics - Bone: dyn_hair_06_01(ComponentPose=@n9, Alpha=1, ExternalForce="0,0,0") {
    angularSpringConstant=100
    bAngularSpring=true
    bUseGravityOverride=true
    simulationSpace=RootRelative
    alphaBoolBlend {blendOption=Linear}
    boundBone {boneName=dyn_hair_06_01}
    componentAppliedLinearAccClamp=<3000 3000 3000>
    componentLinearAccScale=<0.2 0.2 0.2>
    componentLinearVelScale=<0.2 0.2 0.2>
    gravityOverride {z=-980}
    physicsBodyDefinitions [{
      boundBone {boneName=dyn_hair_06_01}
      boxExtents=<5 25 25>
      constraintSetup {angularLimitsMax=<3 5 3>, angularLimitsMin=<-3 -5 -3>, angularTarget=<1 0 0>}
      localJointOffset=<3 0 0>
     }]
    relativeSpaceBone {boneName=spine_05}
   }
   n11 = Anim Dynamics - Bone: dyn_hair_06_02(ComponentPose=@n10, Alpha=1, ExternalForce="0,0,0") {
    angularSpringConstant=100
    bAngularSpring=true
    bUseGravityOverride=true
    simulationSpace=RootRelative
    alphaBoolBlend {blendOption=Linear}
    boundBone {boneName=dyn_hair_06_02}
    componentAppliedLinearAccClamp=<3000 3000 3000>
    componentLinearAccScale=<0.2 0.2 0.2>
    componentLinearVelScale=<0.2 0.2 0.2>
    gravityOverride {z=-980}
    physicsBodyDefinitions [{
      boundBone {boneName=dyn_hair_06_02}
      boxExtents=<5 25 25>
      constraintSetup {angularLimitsMax=<3 5 3>, angularLimitsMin=<-3 -5 -3>, angularTarget=<1 0 0>}
      localJointOffset=<3 0 0>
     }]
    relativeSpaceBone {boneName=spine_05}
   }
   n12 = Anim Dynamics - Bone: dyn_hair_07_01(ComponentPose=@n11, Alpha=1, ExternalForce="0,0,0") {
    angularSpringConstant=100
    bAngularSpring=true
    bUseGravityOverride=true
    simulationSpace=RootRelative
    alphaBoolBlend {blendOption=Linear}
    boundBone {boneName=dyn_hair_07_01}
    componentAppliedLinearAccClamp=<3000 3000 3000>
    componentLinearAccScale=<0.2 0.2 0.2>
    componentLinearVelScale=<0.2 0.2 0.2>
    gravityOverride {z=-980}
    physicsBodyDefinitions [{
      boundBone {boneName=dyn_hair_07_01}
      boxExtents=<5 25 25>
      constraintSetup {angularLimitsMax=<6 9 3>, angularLimitsMin=<-6 -9 -3>, angularTarget=<1 0 0>}
      localJointOffset=<3 0 0>
     }]
    relativeSpaceBone {boneName=spine_05}
   }
   n13 = Anim Dynamics - Bone: dyn_hair_08_01(ComponentPose=@n12, Alpha=1, ExternalForce="0,0,0") {
    angularSpringConstant=100
    bAngularSpring=true
    bUseGravityOverride=true
    simulationSpace=RootRelative
    alphaBoolBlend {blendOption=Linear}
    boundBone {boneName=dyn_hair_08_01}
    componentAppliedLinearAccClamp=<3000 3000 3000>
    componentLinearAccScale=<0.2 0.2 0.2>
    componentLinearVelScale=<0.2 0.2 0.2>
    gravityOverride {z=-980}
    physicsBodyDefinitions [{
      boundBone {boneName=dyn_hair_08_01}
      boxExtents=<5 25 25>
      constraintSetup {angularLimitsMax=<3 9 3>, angularLimitsMin=<-3 -9 -3>, angularTarget=<1 0 0>}
      localJointOffset=<3 0 0>
     }]
    relativeSpaceBone {boneName=spine_05}
   }
   n14 = Anim Dynamics - Bone: dyn_hair_09_01(ComponentPose=@n13, Alpha=1, ExternalForce="0,0,0") {
    angularSpringConstant=100
    bAngularSpring=true
    bUseGravityOverride=true
    simulationSpace=RootRelative
    alphaBoolBlend {blendOption=Linear}
    boundBone {boneName=dyn_hair_09_01}
    componentAppliedLinearAccClamp=<3000 3000 3000>
    componentLinearAccScale=<0.2 0.2 0.2>
    componentLinearVelScale=<0.2 0.2 0.2>
    gravityOverride {z=-980}
    physicsBodyDefinitions [{
      boundBone {boneName=dyn_hair_09_01}
      boxExtents=<5 25 25>
      constraintSetup {angularLimitsMax=<3 9 3>, angularLimitsMin=<-3 -9 -3>, angularTarget=<1 0 0>}
      localJointOffset=<3 0 0>
     }]
    relativeSpaceBone {boneName=spine_05}
   }
   n15 = Anim Dynamics - Bone: dyn_hair_10_01(ComponentPose=@n14, Alpha=1, ExternalForce="0,0,0") {
    angularSpringConstant=100
    bAngularSpring=true
    bUseGravityOverride=true
    simulationSpace=RootRelative
    alphaBoolBlend {blendOption=Linear}
    boundBone {boneName=dyn_hair_10_01}
    componentAppliedLinearAccClamp=<3000 3000 3000>
    componentLinearAccScale=<0.2 0.2 0.2>
    componentLinearVelScale=<0.2 0.2 0.2>
    gravityOverride {z=-980}
    physicsBodyDefinitions [{
      boundBone {boneName=dyn_hair_10_01}
      boxExtents=<5 25 25>
      constraintSetup {angularLimitsMax=<3 9 3>, angularLimitsMin=<-3 -9 -3>, angularTarget=<1 0 0>}
      localJointOffset=<3 0 0>
     }]
    relativeSpaceBone {boneName=spine_05}
   }
   n16 = Anim Dynamics - Bone: dyn_hair_11_01(ComponentPose=@n15, Alpha=1, ExternalForce="0,0,0") {
    angularSpringConstant=100
    bAngularSpring=true
    bUseGravityOverride=true
    simulationSpace=RootRelative
    alphaBoolBlend {blendOption=Linear}
    boundBone {boneName=dyn_hair_11_01}
    componentAppliedLinearAccClamp=<3000 3000 3000>
    componentLinearAccScale=<0.2 0.2 0.2>
    componentLinearVelScale=<0.2 0.2 0.2>
    gravityOverride {z=-980}
    physicsBodyDefinitions [{
      boundBone {boneName=dyn_hair_11_01}
      boxExtents=<5 25 25>
      constraintSetup {angularLimitsMax=<3 3 3>, angularLimitsMin=<-3 -13 -3>, angularTarget=<1 0 0>}
      localJointOffset=<3 0 0>
     }]
    relativeSpaceBone {boneName=spine_05}
   }
   n17 = Anim Dynamics - Bone: dyn_hair_11_02(ComponentPose=@n16, Alpha=1, ExternalForce="0,0,0") {
    angularSpringConstant=100
    bAngularSpring=true
    bUseGravityOverride=true
    simulationSpace=RootRelative
    alphaBoolBlend {blendOption=Linear}
    boundBone {boneName=dyn_hair_11_02}
    componentAppliedLinearAccClamp=<3000 3000 3000>
    componentLinearAccScale=<0.2 0.2 0.2>
    componentLinearVelScale=<0.2 0.2 0.2>
    gravityOverride {z=-980}
    physicsBodyDefinitions [{
      boundBone {boneName=dyn_hair_11_02}
      boxExtents=<5 25 25>
      constraintSetup {angularLimitsMax=<3 3 3>, angularLimitsMin=<-3 -13 -3>, angularTarget=<1 0 0>}
      localJointOffset=<3 0 0>
     }]
    relativeSpaceBone {boneName=spine_05}
   }
   n18 = Anim Dynamics - Bone: dyn_hair_12_01(ComponentPose=@n17, Alpha=1, ExternalForce="0,0,0") {
    angularSpringConstant=100
    bAngularSpring=true
    bUseGravityOverride=true
    simulationSpace=RootRelative
    alphaBoolBlend {blendOption=Linear}
    boundBone {boneName=dyn_hair_12_01}
    componentAppliedLinearAccClamp=<3000 3000 3000>
    componentLinearAccScale=<0.2 0.2 0.2>
    componentLinearVelScale=<0.2 0.2 0.2>
    gravityOverride {z=-980}
    physicsBodyDefinitions [{
      boundBone {boneName=dyn_hair_12_01}
      boxExtents=<5 25 25>
      constraintSetup {angularLimitsMax=<3 9 3>, angularLimitsMin=<-3 -9 -3>, angularTarget=<1 0 0>}
      localJointOffset=<3 0 0>
     }]
    relativeSpaceBone {boneName=spine_05}
   }
   n19 = Anim Dynamics - Bone: dyn_hair_13_01(ComponentPose=@n18, Alpha=1, ExternalForce="0,0,0") {
    angularSpringConstant=100
    bAngularSpring=true
    bUseGravityOverride=true
    simulationSpace=RootRelative
    alphaBoolBlend {blendOption=Linear}
    boundBone {boneName=dyn_hair_13_01}
    componentAppliedLinearAccClamp=<3000 3000 3000>
    componentLinearAccScale=<0.2 0.2 0.2>
    componentLinearVelScale=<0.2 0.2 0.2>
    gravityOverride {z=-980}
    physicsBodyDefinitions [{
      boundBone {boneName=dyn_hair_13_01}
      boxExtents=<5 25 25>
      constraintSetup {angularLimitsMax=<3 8 5>, angularLimitsMin=<-3 -8 -5>, angularTarget=<1 0 0>}
      localJointOffset=<3 0 0>
     }]
    relativeSpaceBone {boneName=spine_05}
   }
   n20 = Anim Dynamics - Bone: dyn_hair_14_01(ComponentPose=@n19, Alpha=1, ExternalForce="0,0,0") {
    angularSpringConstant=100
    bAngularSpring=true
    bUseGravityOverride=true
    simulationSpace=RootRelative
    alphaBoolBlend {blendOption=Linear}
    boundBone {boneName=dyn_hair_14_01}
    componentAppliedLinearAccClamp=<3000 3000 3000>
    componentLinearAccScale=<0.2 0.2 0.2>
    componentLinearVelScale=<0.2 0.2 0.2>
    gravityOverride {z=-980}
    physicsBodyDefinitions [{
      boundBone {boneName=dyn_hair_14_01}
      boxExtents=<5 25 25>
      constraintSetup {angularLimitsMax=<3 4 3>, angularLimitsMin=<-3 -3 -3>, angularTarget=<1 0 0>}
      localJointOffset=<3 0 0>
     }]
    relativeSpaceBone {boneName=spine_05}
   }
   n21 = Anim Dynamics - Bone: dyn_hair_14_02(ComponentPose=@n20, Alpha=1, ExternalForce="0,0,0") {
    angularSpringConstant=100
    bAngularSpring=true
    bUseGravityOverride=true
    simulationSpace=RootRelative
    alphaBoolBlend {blendOption=Linear}
    boundBone {boneName=dyn_hair_14_02}
    componentAppliedLinearAccClamp=<3000 3000 3000>
    componentLinearAccScale=<0.2 0.2 0.2>
    componentLinearVelScale=<0.2 0.2 0.2>
    gravityOverride {z=-980}
    physicsBodyDefinitions [{
      boundBone {boneName=dyn_hair_14_02}
      boxExtents=<5 25 25>
      constraintSetup {angularLimitsMax=<3 7 3>, angularLimitsMin=<-3 -3 -3>, angularTarget=<1 0 0>}
      localJointOffset=<3 0 0>
     }]
    relativeSpaceBone {boneName=spine_05}
   }
   n22 = Anim Dynamics - Bone: dyn_hair_15_01(ComponentPose=@n21, Alpha=1, ExternalForce="0,0,0") {
    angularSpringConstant=100
    bAngularSpring=true
    bUseGravityOverride=true
    simulationSpace=RootRelative
    alphaBoolBlend {blendOption=Linear}
    boundBone {boneName=dyn_hair_15_01}
    componentAppliedLinearAccClamp=<3000 3000 3000>
    componentLinearAccScale=<0.2 0.2 0.2>
    componentLinearVelScale=<0.2 0.2 0.2>
    gravityOverride {z=-980}
    physicsBodyDefinitions [{
      boundBone {boneName=dyn_hair_15_01}
      boxExtents=<5 25 25>
      constraintSetup {angularLimitsMax=<3 3 3>, angularLimitsMin=<-3 -12 -3>, angularTarget=<1 0 0>}
      localJointOffset=<3 0 0>
     }]
    relativeSpaceBone {boneName=spine_05}
   }
   n23 = Anim Dynamics - Bone: dyn_hair_15_02(ComponentPose=@n22, Alpha=1, ExternalForce="0,0,0") {
    angularSpringConstant=100
    bAngularSpring=true
    bUseGravityOverride=true
    simulationSpace=RootRelative
    alphaBoolBlend {blendOption=Linear}
    boundBone {boneName=dyn_hair_15_02}
    componentAppliedLinearAccClamp=<3000 3000 3000>
    componentLinearAccScale=<0.2 0.2 0.2>
    componentLinearVelScale=<0.2 0.2 0.2>
    gravityOverride {z=-980}
    physicsBodyDefinitions [{
      boundBone {boneName=dyn_hair_15_02}
      boxExtents=<5 25 25>
      constraintSetup {angularLimitsMax=<3 5 3>, angularLimitsMin=<-3 -4 -3>, angularTarget=<1 0 0>}
      localJointOffset=<3 0 0>
     }]
    relativeSpaceBone {boneName=spine_05}
   }
   n24 = Anim Dynamics - Bone: dyn_hair_16_01(ComponentPose=@n23, Alpha=1, ExternalForce="0,0,0") {
    angularSpringConstant=100
    bAngularSpring=true
    bUseGravityOverride=true
    simulationSpace=RootRelative
    alphaBoolBlend {blendOption=Linear}
    boundBone {boneName=dyn_hair_16_01}
    componentAppliedLinearAccClamp=<3000 3000 3000>
    componentLinearAccScale=<0.2 0.2 0.2>
    componentLinearVelScale=<0.2 0.2 0.2>
    gravityOverride {z=-980}
    physicsBodyDefinitions [{
      boundBone {boneName=dyn_hair_16_01}
      boxExtents=<5 25 25>
      constraintSetup {angularLimitsMax=<3 3 3>, angularLimitsMin=<-3 -14 -3>, angularTarget=<1 0 0>}
      localJointOffset=<3 0 0>
     }]
    relativeSpaceBone {boneName=spine_05}
   }
   n25 = Anim Dynamics - Bone: dyn_hair_back_01(ComponentPose=@n24, Alpha=1, ExternalForce="0,0,0") {
    angularSpringConstant=100
    bAngularSpring=true
    bUseGravityOverride=true
    simulationSpace=RootRelative
    alphaBoolBlend {blendOption=Linear}
    boundBone {boneName=dyn_hair_back_01}
    componentAppliedLinearAccClamp=<3000 3000 3000>
    componentLinearAccScale=<0.2 0.2 0.2>
    componentLinearVelScale=<0.2 0.2 0.2>
    gravityOverride {z=-980}
    physicsBodyDefinitions [{
      boundBone {boneName=dyn_hair_back_01}
      boxExtents=<12 50 50>
      constraintSetup {angularLimitsMax=<0 1 0>, angularLimitsMin=<0 -5 0>, angularTarget=<1 0 0>}
      localJointOffset=<5 0 0>
     }]
    relativeSpaceBone {boneName=spine_05}
   }
   Output Pose(Result=Component To Local(ComponentPose=@n25)) {layerGroup=None}
  ]}
 EventGraph {nodes [
   n1 = Event OnBeginFalling()
   Set jumping(execute=[@n1.then, @n3.then], jumping:double=1) -> (Output_Get:double)
   n2 = Event OnEndFalling()
   Set jumping(execute=[@n2.then, @n4.then], jumping=0)
   n3 = Event OnBeginJumping()
   n4 = Event OnEndJumping()
   Event BlueprintUpdateAnimation()
  ]}
]
variables [jumping {type=number}]
